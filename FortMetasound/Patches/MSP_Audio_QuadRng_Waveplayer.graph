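# Reusable MetaSound patch that plays a random four-channel wave from an input array with a randomized pitch shift.
# Play triggers a random array get with no-repeat and optional shared state, which then triggers a random float between Pitch.Min and Pitch.Max.
# That value drives the quad Wave Player's pitch shift as it starts.
# The patch exposes loop controls, front and side audio pairs, and the player's playback and cue point outputs.
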MetaSoundPatch /FortMetasound/Patches/MSP_Audio_QuadRng_Waveplayer
inputs [
 Enable Shared State {type=Bool}
 In Array {type=WaveAsset:Array}
 Loop {type=Bool}
 Loop Duration {type=Time}
 Loop Start {type=Time}
 No Repeats {type=Int32}
 Pitch.Max {type=Float}
 Pitch.Min {type=Float}
 Play {type=Trigger}
 Reset {type=Trigger}
 Start Time {type=Time}
 Stop {type=Trigger}
]
outputs [
 Cue Point ID {type=Int32}
 Cue Point Label {type=String}
 Loop Percent {type=Float}
 On Cue Point {type=Trigger}
 On Finished {type=Trigger}
 On Looped {type=Trigger}
 On Nearly Finished {type=Trigger}
 On Play {type=Trigger}
 Out Side Left {type=Audio}
 Out Side Right {type=Audio}
 Out.L {type=Audio}
 Out.R {type=Audio}
 Playback Location {type=Float}
 Playback Time {type=Time}
]
pages [Default {
  input_defaults {
   Enable Shared State=null
   In Array=null
   Loop=null
   Loop Duration=-1
   Loop Start=0
   No Repeats=1
   Pitch.Max=1
   Pitch.Min=0
   Play=null
   Reset=null
   Start Time=0
   Stop=null
  }
  nodes [
   ArrayRandomGet_0 = Array.Random Get:WaveAsset:Array(Enable Shared State=^Enable Shared State, In Array=^In Array, No Repeats=^No Repeats, Next=^Play, Reset=^Reset)
   n1 = UE.RandomFloat(Max=^Pitch.Max, Min=^Pitch.Min, Next=@ArrayRandomGet_0.On Next, Reset=@ArrayRandomGet_0.On Reset)
   WavePlayer_0 = UE.Wave Player:Quad(
    Loop=^Loop
    Loop Duration=^Loop Duration
    Loop Start=^Loop Start
    Start Time=^Start Time
    Stop=^Stop
    Wave Asset=@ArrayRandomGet_0.Value
    Play=@n1.On Next
    Pitch Shift=@n1.Value
   ) {also {version [1,1]}}
  ]
  links [
   {dst="@Cue Point ID", src=WavePlayer_0.Cue Point ID}
   {dst="@Cue Point Label", src=WavePlayer_0.Cue Point Label}
   {dst="@Loop Percent", src=WavePlayer_0.Loop Percent}
   {dst="@On Cue Point", src=WavePlayer_0.On Cue Point}
   {dst="@On Finished", src=WavePlayer_0.On Finished}
   {dst="@On Looped", src=WavePlayer_0.On Looped}
   {dst="@On Nearly Finished", src=WavePlayer_0.On Nearly Finished}
   {dst="@On Play", src=WavePlayer_0.On Play}
   {dst="@Out.L", src=WavePlayer_0.Out Front Left}
   {dst="@Out.R", src=WavePlayer_0.Out Front Right}
   {dst="@Out Side Left", src=WavePlayer_0.Out Side Left}
   {dst="@Out Side Right", src=WavePlayer_0.Out Side Right}
   {dst="@Playback Location", src=WavePlayer_0.Playback Location}
   {dst="@Playback Time", src=WavePlayer_0.Playback Time}
  ]
 }]
