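# MetaSound patch that multiplies six float inputs into a single float output.
# A chain of two-input Multiply nodes folds Input1 through Input6 together in order and sends the product to Out.
# Every input defaults to 1, so unconnected inputs leave the result unchanged, making it a reusable multi-operand multiply for other MetaSounds.
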MetaSoundPatch /FortMetasound/Patches/MSP_BigMultiplyFloat
inputs [
 Input1 {type=Float}
 Input2 {type=Float}
 Input3 {type=Float}
 Input4 {type=Float}
 Input5 {type=Float}
 Input6 {type=Float}
]
interfaces [MetaSound:1.0]
outputs [Out {type=Float}]
pages [Default {
  input_defaults {Input1=1, Input2=1, Input3=1, Input4=1, Input5=1, Input6=1}
  nodes [
   n1 = UE.Multiply:Float(AdditionalOperands=^Input3, PrimaryOperand=UE.Multiply:Float(PrimaryOperand=^Input1, AdditionalOperands=^Input2).Out)
   Multiply_0 = UE.Multiply:Float(
    AdditionalOperands=^Input6
    PrimaryOperand=UE.Multiply:Float(AdditionalOperands=^Input5, PrimaryOperand=UE.Multiply:Float(AdditionalOperands=^Input4, PrimaryOperand=@n1.Out).Out).Out
   )
  ]
  links [{dst="@Out", src=Multiply_0.Out}]
 }]
