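# MSP_BoolToFloatValues is a reusable MetaSound patch that turns a Bool input into one of two configurable float values, smoothed over time.
# The Bool is converted to a float, which MapRange remaps between FalseValue and TrueValue.
# InterpTo then glides toward that target over Interp Time and drives Out Value, with defaults of zero to one and zero interpolation time.
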MetaSoundPatch /FortMetasound/Patches/MSP_BoolToFloatValues
inputs [
 FalseValue {type=Float}
 Input {type=Bool}
 Interp Time {type=Time}
 TrueValue {type=Float}
]
interfaces [MetaSound:1.0]
outputs [Out Value {type=Float}]
pages [Default {
  input_defaults {FalseValue=0, Input=null, Interp Time=0, TrueValue=1}
  nodes [InterpTo_0 = UE.InterpTo:Audio(
    Interp Time=^Interp Time
    Target=MapRange.MapRange:Float(Out Range A=^FalseValue, Out Range B=^TrueValue, In=Convert.Float:Bool(Bool=^Input).Float).Out Value
   )]
  links [{dst="@Out Value", src=InterpTo_0.Value}]
 }]
