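# MetaSound patch that high-pass filters an audio input according to an Elevation value, presumably the sound's vertical angle in degrees.
# A MapRange converts Elevation from -90 to 90 into a cutoff between the BellowCutOffFreq and AboveCutOffFreq inputs.
# That cutoff drives a one-pole high-pass filter, so with default inputs more low end is removed as elevation rises.
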MetaSoundPatch /FortMetasound/Patches/MSP_ElevationHighPass
inputs [
 AboveCutOffFreq {type=Float}
 AudioIn {type=Audio}
 BellowCutOffFreq {type=Float}
 Elevation {type=Float}
]
outputs [Audio Out {type=Audio}]
pages [Default {
  input_defaults {AboveCutOffFreq=100, AudioIn=null, BellowCutOffFreq=0, Elevation=0}
  nodes [OnePoleHighPassFilter_0 = UE.One-Pole High Pass Filter:Audio(
    In=^AudioIn
    Cutoff Frequency=MapRange.MapRange:Float(Out Range B=^AboveCutOffFreq, Out Range A=^BellowCutOffFreq, In=^Elevation) {also {defaults {In Range A=-90, In Range B=90}}}.Out Value
   )]
  links [{dst="@Audio Out", src=OnePoleHighPassFilter_0.Out}]
 }]
