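# MetaSound patch that logs debug info about an input Wave Asset when triggered, with no outputs.
# A trigger compare on bBypass gates the incoming trigger, then fans out to three more compares, one each for the name, path and duration toggles.
# Each passing compare fires its own Print Log with the shared Label, fed by a Get Wave Duration node, with duration converted from time to float.
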MetaSoundPatch /FortMetasound/Patches/MSP_PrintWaveInfo
inputs [
 bBypass {type=Bool}
 bPrintWaveAssetName {type=Bool}
 bPrintWaveAssetPath {type=Bool}
 Input {type=WaveAsset}
 Label {type=String}
 PrintWaveAssetDuration {type=Bool}
 Trigger {type=Trigger}
]
outputs []
pages [Default {
  input_defaults {
   Input=null
   Label=Print Wave:
   PrintWaveAssetDuration=null
   Trigger=null
   bBypass=null
   bPrintWaveAssetName="<MetaSoundBuilderResult.SUCCEEDED: 0>"
   bPrintWaveAssetPath=null
  }
  nodes [
   GetWaveDuration_0 = "UE.Get Wave Duration: "(Wave=^Input)
   n1 = TriggerCompare.Clamp:Bool(Compare=^Trigger, A=^bBypass) {also {defaults {B="<MetaSoundBuilderResult.SUCCEEDED: 0>", Type=1}}}
   PrintLogPrintLog_0 = Print Log.Print Log:Float(
    Label=^Label
    Value To Log=UE.ConversionTimeToFloat(In=@GetWaveDuration_0.Duration).Out
    Trigger=TriggerCompare.Clamp:Bool(A=^PrintWaveAssetDuration, Compare=@n1.True) {also {defaults {B="<MetaSoundBuilderResult.SUCCEEDED: 0>"}}}.True
   )
   PrintLogPrintLog_1 = Print Log.Print Log:String(
    Label=^Label
    Value To Log=@GetWaveDuration_0.Path
    Trigger=TriggerCompare.Clamp:Bool(A=^bPrintWaveAssetPath, Compare=@n1.True) {also {defaults {B="<MetaSoundBuilderResult.SUCCEEDED: 0>"}}}.True
   )
   PrintLogPrintLog_2 = Print Log.Print Log:String(
    Label=^Label
    Value To Log=@GetWaveDuration_0.Name
    Trigger=TriggerCompare.Clamp:Bool(A=^bPrintWaveAssetName, Compare=@n1.True) {also {defaults {B="<MetaSoundBuilderResult.SUCCEEDED: 0>"}}}.True
   )
  ]
 }]
