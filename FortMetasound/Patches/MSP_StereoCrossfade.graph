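# MetaSound patch that crossfades between two stereo audio inputs and outputs a single stereo pair.
# One float input, defaulting to 0, drives two two-input audio crossfade nodes, one per channel.
# The left node blends both left inputs into OutL and the right node blends both right inputs into OutR, so both channels share the same blend position.
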MetaSoundPatch /FortMetasound/Patches/MSP_StereoCrossfade
inputs [
 FloatIn {type=Float}
 Input 1 L {type=Audio}
 Input 1 R {type=Audio}
 Input 2 L {type=Audio}
 Input 2 R {type=Audio}
]
interfaces [MetaSound:1.0]
outputs [
 OutL {type=Audio}
 OutR {type=Audio}
]
pages [Default {
  input_defaults {FloatIn=0, Input 1 L=null, Input 1 R=null, Input 2 L=null, Input 2 R=null}
  nodes [
   CrossfadeTriggerRouteAudio2_0 = "Crossfade.Trigger Route (Audio, 2):Audio"(Crossfade Value=^FloatIn, In 0=^Input 1 R, In 1=^Input 2 R)
   CrossfadeTriggerRouteAudio2_1 = "Crossfade.Trigger Route (Audio, 2):Audio"(Crossfade Value=^FloatIn, In 0=^Input 1 L, In 1=^Input 2 L)
  ]
  links [{dst="@OutR", src=CrossfadeTriggerRouteAudio2_0.Out}, {dst="@OutL", src=CrossfadeTriggerRouteAudio2_1.Out}]
 }]
