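# MSP_StereoSpread is a reusable MetaSound patch that adjusts the stereo width of a left/right audio pair using mid-side processing.
# It derives mid as half the sum of the inputs and side as half their difference.
# Side is scaled by Strength, smoothed over Interp Time, then added to and subtracted from mid for OutL and OutR.
# Strength 0, the default, yields mono, 1 preserves the original image, and higher values widen it.
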MetaSoundPatch /FortMetasound/Patches/MSP_StereoSpread
inputs [
 Interp Time {type=Time}
 Left In {type=Audio}
 Right In {type=Audio}
 Strength {type=Float}
]
outputs [
 OutL {type=Audio}
 OutR {type=Audio}
]
pages [Default {
  input_defaults {Interp Time=0, Left In=null, Right In=null, Strength=0}
  nodes [
   n1 = UE.Multiply:Audio by Float(PrimaryOperand=UE.Add:Audio(PrimaryOperand=^Left In, AdditionalOperands=^Right In).Out) {also {defaults {AdditionalOperands=0.5}}}
   n2 = UE.Multiply:Audio by Float(
    AdditionalOperands=UE.InterpTo:Audio(Interp Time=^Interp Time, Target=^Strength).Value
    PrimaryOperand=UE.Multiply:Audio by Float(PrimaryOperand=UE.Subtract:Audio(PrimaryOperand=^Left In, AdditionalOperands=^Right In) {also {version [1,1]}}.Out) {also {defaults {AdditionalOperands=0.5}}}.Out
   )
   Add_1 = UE.Add:Audio(PrimaryOperand=@n1.Out, AdditionalOperands=@n2.Out)
   Subtract_0 = UE.Subtract:Audio(PrimaryOperand=@n1.Out, AdditionalOperands=@n2.Out) {also {version [1,1]}}
  ]
  links [{dst="@OutL", src=Add_1.Out}, {dst="@OutR", src=Subtract_0.Out}]
 }]
