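# MetaSound patch that turns a float Value into one-shot gate enter and exit triggers by comparing it to a threshold with selectable conditions.
# On each value change, two compares feed Trigger Controls that close on their own output.
# Enter opens the exit control and starts RetriggerDelay; a trigger accumulator reopens enter once both the delay and exit have fired.
# Deferred trigger variables carry the feedback loops.
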MetaSoundPatch /FortMetasound/Patches/MSP_OneshotGate
inputs [
 EnterCondition {type=Enum:TriggerComparisonType}
 ExitCondition {type=Enum:TriggerComparisonType}
 OneshotThreshold {type=Float}
 RetriggerDelay {type=Time}
 Value {type=Float}
]
interfaces [MetaSound:1.0]
outputs [
 OnGateEnter {type=Trigger}
 OnGateExit {type=Trigger}
]
pages [Default {
  input_defaults {EnterCondition=5, ExitCondition=2, OneshotThreshold=0.5, RetriggerDelay=1, Value=1}
  nodes [
   n1 = InitVariable.Trigger()
   n2 = InitVariable.Trigger()
   n3 = VariableDeferredAccessor.Trigger(Variable=@n2.Variable)
   n4 = VariableDeferredAccessor.Trigger(Variable=@n3.Variable)
   TriggerDelay_0 = UE.Trigger Delay(Delay Time=^RetriggerDelay, In=@n4.Value)
   TriggerOnValueChange_0 = UE.Trigger On Value Change:Float(Value=^Value)
   n5 = VariableDeferredAccessor.Trigger(Variable=InitVariable.Trigger().Variable)
   TriggerControl_1 = UE.Trigger Control(
    Trigger In=TriggerCompare.Clamp:Float(Type=^ExitCondition, B=^OneshotThreshold, A=^Value, Compare=@TriggerOnValueChange_0.Trigger).True
    Close=@n5.Value
    Open=VariableDeferredAccessor.Trigger(Variable=@n2.Variable).Value
   )
   TriggerAccumulatorTriggerAccumulate2_0 = "TriggerAccumulator.Trigger Accumulate (2)"(
    In 0=VariableAccessor.Trigger(Variable=VariableMutator.Trigger(Variable=@n1.Variable, Value=@TriggerDelay_0.Out).Variable).Value
    In 1=VariableAccessor.Trigger(Variable=VariableMutator.Trigger(Value=@TriggerControl_1.Trigger Out, Variable=@n5.Variable).Variable).Value
   ) {also {defaults {Auto Reset="<MetaSoundBuilderResult.SUCCEEDED: 0>"}}}
   TriggerOnValueChange_1 = UE.Trigger On Value Change:Float(Value=^Value)
   TriggerControl_0 = UE.Trigger Control(
    Open=@TriggerAccumulatorTriggerAccumulate2_0.Out
    Trigger In=TriggerCompare.Clamp:Float(Type=^EnterCondition, B=^OneshotThreshold, A=^Value, Compare=@TriggerOnValueChange_1.Trigger).True
    Close=@n3.Value
   )
   VariableMutator.Trigger(Value=@TriggerControl_0.Trigger Out, Variable=@n4.Variable)
  ]
  links [{dst="@OnGateEnter", src=TriggerControl_0.Trigger Out}, {dst="@OnGateExit", src=TriggerControl_1.Trigger Out}]
 }]
