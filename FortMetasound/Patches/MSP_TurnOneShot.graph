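# MSP_TurnOneShot is a MetaSound patch that fires a TriggerTurn trigger when turning speed passes a threshold, for one-shot turn sounds.
# It maps the signed Movement.SpeedAzimuth input to a normalized range and takes the absolute value, so both directions count equally.
# That value feeds an MSP_OneshotGate with threshold and retrigger-delay inputs.
# The gate's enter event passes through MSP_Branch, which emits TriggerTurn only while AllowOneShot is true.
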MetaSoundPatch /FortMetasound/Patches/MSP_TurnOneShot
inputs [
 AllowOneShot {type=Bool}
 Movement.SpeedAzimuth {type=Float}
 OneshotThreshold {type=Float}
 RetriggerDelay {type=Time}
]
outputs [TriggerTurn {type=Trigger}]
pages [Default {
  input_defaults {AllowOneShot=null, Movement.SpeedAzimuth=0, OneshotThreshold=0.75, RetriggerDelay=1.5}
  nodes [
   InitVariable.Audio()
   InitVariable.Audio()
   InitVariable.Bool()
   n1 = MapRange.MapRange:Float(In=^Movement.SpeedAzimuth) {also {defaults {In Range A=-180, In Range B=180, Out Range A=-1}}}
   MSP_Branch_0 = &MSP_Branch(
    Value=^AllowOneShot
    In=&MSP_OneshotGate(OneshotThreshold=^OneshotThreshold, RetriggerDelay=^RetriggerDelay, Value=AbsoluteValue.Abs:Float(Input=@n1.Out Value).Value) {also {defaults {EnterCondition=4, ExitCondition=3}}}.OnGateEnter
   )
  ]
  links [{dst="@TriggerTurn", src=MSP_Branch_0.True}]
 }]
