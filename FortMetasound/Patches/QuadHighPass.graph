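# MetaSound patch that applies a high-pass filter to four-channel quad audio (FL, FR, SL, SR).
# Each channel runs through its own One-Pole High Pass Filter to its matching output.
# All four filters share a single Cutoff Frequency input, which defaults to 100.
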MetaSoundPatch /FortMetasound/Patches/QuadHighPass
inputs [
 Cutoff Frequency {type=Float}
 In FL {type=Audio}
 In FR {type=Audio}
 In SL {type=Audio}
 In SR {type=Audio}
]
outputs [
 Out FL {type=Audio}
 Out FR {type=Audio}
 Out SL {type=Audio}
 Out SR {type=Audio}
]
pages [Default {
  input_defaults {Cutoff Frequency=100, In FL=null, In FR=null, In SL=null, In SR=null}
  nodes [
   OnePoleHighPassFilter_0 = UE.One-Pole High Pass Filter:Audio(Cutoff Frequency=^Cutoff Frequency, In=^In SR)
   OnePoleHighPassFilter_1 = UE.One-Pole High Pass Filter:Audio(Cutoff Frequency=^Cutoff Frequency, In=^In FL)
   OnePoleHighPassFilter_2 = UE.One-Pole High Pass Filter:Audio(Cutoff Frequency=^Cutoff Frequency, In=^In FR)
   OnePoleHighPassFilter_3 = UE.One-Pole High Pass Filter:Audio(Cutoff Frequency=^Cutoff Frequency, In=^In SL)
  ]
  links [
   {dst="@Out SR", src=OnePoleHighPassFilter_0.Out}
   {dst="@Out FL", src=OnePoleHighPassFilter_1.Out}
   {dst="@Out FR", src=OnePoleHighPassFilter_2.Out}
   {dst="@Out SL", src=OnePoleHighPassFilter_3.Out}
  ]
 }]
