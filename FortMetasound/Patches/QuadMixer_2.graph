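# QuadMixer_2 is a MetaSound patch that mixes two quad audio sources into one quad output.
# Two stereo 2-input Audio Mixer nodes split the work: one sums the front pairs to Output L and R, the other the SL/SR pairs to Output SL and SR.
# Each source has separate gains for its front and SL/SR pairs, each shared by left and right and defaulting to unity.
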MetaSoundPatch /FortMetasound/Patches/QuadMixer_2
inputs [
 Gain FLR 1 {type=Float}
 Gain FLR 2 {type=Float}
 Gain SLR 1 {type=Float}
 Gain SLR 2 {type=Float}
 Input FL 1 {type=Audio}
 Input FL 2 {type=Audio}
 Input FR 1 {type=Audio}
 Input FR 2 {type=Audio}
 Input SL 1 {type=Audio}
 Input SL 2 {type=Audio}
 Input SR 1 {type=Audio}
 Input SR 2 {type=Audio}
]
outputs [
 Output L {type=Audio}
 Output R {type=Audio}
 Output SL {type=Audio}
 Output SR {type=Audio}
]
pages [Default {
  input_defaults {
   Gain FLR 1=1
   Gain FLR 2=1
   Gain SLR 1=1
   Gain SLR 2=1
   Input FL 1=null
   Input FL 2=null
   Input FR 1=null
   Input FR 2=null
   Input SL 1=null
   Input SL 2=null
   Input SR 1=null
   Input SR 2=null
  }
  nodes [
   AudioMixerAudioMixerStereo2_0 = "AudioMixer.Audio Mixer (Stereo, 2)"(Gain 0=^Gain SLR 1, Gain 1=^Gain SLR 2, In 0 L=^Input SL 1, In 1 L=^Input SL 2, In 0 R=^Input SR 1, In 1 R=^Input SR 2)
   AudioMixerAudioMixerStereo2_1 = "AudioMixer.Audio Mixer (Stereo, 2)"(Gain 0=^Gain FLR 1, Gain 1=^Gain FLR 2, In 0 L=^Input FL 1, In 1 L=^Input FL 2, In 0 R=^Input FR 1, In 1 R=^Input FR 2)
  ]
  links [
   {dst="@Output SL", src=AudioMixerAudioMixerStereo2_0.Out L}
   {dst="@Output SR", src=AudioMixerAudioMixerStereo2_0.Out R}
   {dst="@Output L", src=AudioMixerAudioMixerStereo2_1.Out L}
   {dst="@Output R", src=AudioMixerAudioMixerStereo2_1.Out R}
  ]
 }]
