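# MetaSound patch that mixes four quad sources (front L/R, surround L/R) into one quad output with per-source gain inputs.
# Cascaded two-input stereo mixers sum sources in pairs, separately for front and surround.
# Gain wiring is crossed for sources 3 and 4: Gain SLR 3 scales front 4, and Gain FLR 4 scales surround 3.
# Surround source 3 also has left and right swapped.
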MetaSoundPatch /FortMetasound/Patches/QuadMixer_4
inputs [
 Gain FLR 1 {type=Float}
 Gain FLR 2 {type=Float}
 Gain FLR 3 {type=Float}
 Gain FLR 4 {type=Float}
 Gain SLR 1 {type=Float}
 Gain SLR 2 {type=Float}
 Gain SLR 3 {type=Float}
 Gain SLR 4 {type=Float}
 Input FL 1 {type=Audio}
 Input FL 2 {type=Audio}
 Input FL 3 {type=Audio}
 Input FL 4 {type=Audio}
 Input FR 1 {type=Audio}
 Input FR 2 {type=Audio}
 Input FR 3 {type=Audio}
 Input FR 4 {type=Audio}
 Input SL 1 {type=Audio}
 Input SL 2 {type=Audio}
 Input SL 3 {type=Audio}
 Input SL 4 {type=Audio}
 Input SR 1 {type=Audio}
 Input SR 2 {type=Audio}
 Input SR 3 {type=Audio}
 Input SR 4 {type=Audio}
]
outputs [
 Output L {type=Audio}
 Output R {type=Audio}
 Output SL {type=Audio}
 Output SR {type=Audio}
]
pages [Default {
  input_defaults {
   Gain FLR 1=1
   Gain FLR 2=1
   Gain FLR 3=1
   Gain FLR 4=1
   Gain SLR 1=1
   Gain SLR 2=1
   Gain SLR 3=1
   Gain SLR 4=1
   Input FL 1=null
   Input FL 2=null
   Input FL 3=null
   Input FL 4=null
   Input FR 1=null
   Input FR 2=null
   Input FR 3=null
   Input FR 4=null
   Input SL 1=null
   Input SL 2=null
   Input SL 3=null
   Input SL 4=null
   Input SR 1=null
   Input SR 2=null
   Input SR 3=null
   Input SR 4=null
  }
  nodes [
   AudioMixerAudioMixerStereo2_4 = "AudioMixer.Audio Mixer (Stereo, 2)"(Gain 0=^Gain FLR 1, Gain 1=^Gain FLR 2, In 0 L=^Input FL 1, In 1 L=^Input FL 2, In 0 R=^Input FR 1, In 1 R=^Input FR 2)
   AudioMixerAudioMixerStereo2_5 = "AudioMixer.Audio Mixer (Stereo, 2)"(Gain 0=^Gain FLR 3, Gain 1=^Gain SLR 3, In 0 L=^Input FL 3, In 1 L=^Input FL 4, In 0 R=^Input FR 3, In 1 R=^Input FR 4)
   AudioMixerAudioMixerStereo2_1 = "AudioMixer.Audio Mixer (Stereo, 2)"(
    In 0 L=@AudioMixerAudioMixerStereo2_4.Out L
    In 0 R=@AudioMixerAudioMixerStereo2_4.Out R
    In 1 L=@AudioMixerAudioMixerStereo2_5.Out L
    In 1 R=@AudioMixerAudioMixerStereo2_5.Out R
   )
   AudioMixerAudioMixerStereo2_0 = "AudioMixer.Audio Mixer (Stereo, 2)"(Gain 0=^Gain SLR 1, Gain 1=^Gain SLR 2, In 0 L=^Input SL 1, In 1 L=^Input SL 2, In 0 R=^Input SR 1, In 1 R=^Input SR 2)
   AudioMixerAudioMixerStereo2_3 = "AudioMixer.Audio Mixer (Stereo, 2)"(Gain 0=^Gain FLR 4, Gain 1=^Gain SLR 4, In 0 R=^Input SL 3, In 1 L=^Input SL 4, In 0 L=^Input SR 3, In 1 R=^Input SR 4)
   AudioMixerAudioMixerStereo2_2 = "AudioMixer.Audio Mixer (Stereo, 2)"(
    In 0 L=@AudioMixerAudioMixerStereo2_0.Out L
    In 0 R=@AudioMixerAudioMixerStereo2_0.Out R
    In 1 L=@AudioMixerAudioMixerStereo2_3.Out L
    In 1 R=@AudioMixerAudioMixerStereo2_3.Out R
   )
  ]
  links [
   {dst="@Output L", src=AudioMixerAudioMixerStereo2_1.Out L}
   {dst="@Output R", src=AudioMixerAudioMixerStereo2_1.Out R}
   {dst="@Output SL", src=AudioMixerAudioMixerStereo2_2.Out L}
   {dst="@Output SR", src=AudioMixerAudioMixerStereo2_2.Out R}
  ]
 }]
