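# MetaSound patch QuadMultFloat scales a four-channel quad audio signal (FL, FR, SL, SR) by a single Float input.
# Four Multiply (Audio by Float) nodes each take one channel and the shared Float, then feed the matching output.
# The multiplier defaults to 1, so audio passes through at unchanged level unless that input is set.
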MetaSoundPatch /FortMetasound/Patches/QuadMultFloat
inputs [
 Float {type=Float}
 In FL {type=Audio}
 In FR {type=Audio}
 In SL {type=Audio}
 In SR {type=Audio}
]
outputs [
 Out FL {type=Audio}
 Out FR {type=Audio}
 Out SL {type=Audio}
 Out SR {type=Audio}
]
pages [Default {
  input_defaults {Float=1, In FL=null, In FR=null, In SL=null, In SR=null}
  nodes [
   Multiply_0 = UE.Multiply:Audio by Float(AdditionalOperands=^Float, PrimaryOperand=^In SL)
   Multiply_1 = UE.Multiply:Audio by Float(AdditionalOperands=^Float, PrimaryOperand=^In SR)
   Multiply_2 = UE.Multiply:Audio by Float(AdditionalOperands=^Float, PrimaryOperand=^In FL)
   Multiply_3 = UE.Multiply:Audio by Float(AdditionalOperands=^Float, PrimaryOperand=^In FR)
  ]
  links [
   {dst="@Out SL", src=Multiply_0.Out}
   {dst="@Out SR", src=Multiply_1.Out}
   {dst="@Out FL", src=Multiply_2.Out}
   {dst="@Out FR", src=Multiply_3.Out}
  ]
 }]
