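# MetaSound patch that acts as a reusable float-based switch, routing an incoming trigger to either a True or False trigger output.
# A single Trigger Compare node tests FloatIn against SwitchValue whenever TriggerIn fires.
# The comparison operator is exposed as the ComparisonType input, so each use of the patch can choose its own test.
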MetaSoundPatch /FortMetasound/Patches/SwitchViaFloat
inputs [
 ComparisonType {type=Enum:TriggerComparisonType}
 FloatIn {type=Float}
 SwitchValue {type=Float}
 TriggerIn {type=Trigger}
]
interfaces [MetaSound:1.0]
outputs [
 False {type=Trigger}
 True {type=Trigger}
]
pages [Default {
  input_defaults {ComparisonType=3, FloatIn=0, SwitchValue=0, TriggerIn=null}
  nodes [TriggerCompareClamp_0 = TriggerCompare.Clamp:Float(Type=^ComparisonType, A=^FloatIn, B=^SwitchValue, Compare=^TriggerIn)]
  links [{dst="@False", src=TriggerCompareClamp_0.FALSE}, {dst="@True", src=TriggerCompareClamp_0.TRUE}]
 }]
