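# MetaSound patch that pans a stereo input around the listener by azimuth, folding the full circle onto a stereo output.
# Right and left inputs sit at azimuth plus and minus StereoSpread, each wrapped into ±180 degrees by trigger-compare logic adding or subtracting 360.
# A five-point float crossfade maps each angle to pan: centered front and back, hard at ±90.
# Two stereo panners apply these values and a mixer sums them.
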MetaSoundPatch /FortMetasound/Patches/AzimuthPannerStereo
inputs [
 Azimuth {type=Float}
 In L {type=Audio}
 In R {type=Audio}
 StereoSpread {type=Float}
]
interfaces [MetaSound:1.0]
outputs [
 Out L {type=Audio}
 Out R {type=Audio}
]
pages [Default {
  input_defaults {Azimuth=0, In L=null, In R=null, StereoSpread=0}
  nodes [
   n1 = UE.Add:Float(PrimaryOperand=^Azimuth, AdditionalOperands=^StereoSpread)
   TriggerOnValueChange_1 = UE.Trigger On Value Change:Float(Value=@n1.Out)
   n2 = TriggerCompare.Clamp:Float(A=@n1.Out, Compare=@TriggerOnValueChange_1.Trigger) {also {defaults {B=180, Type=3}}}
   TriggerRouteTriggerRouteBool2_0 = "TriggerRoute.Trigger Route (Bool, 2):Bool"(Set 1=@n2.False, Set 0=@n2.True) {also {defaults {Value 0="<MetaSoundBuilderResult.SUCCEEDED: 0>"}}}
   n3 = UE.Add:Float(
    PrimaryOperand=@n1.Out
    AdditionalOperands=UE.Multiply:Float(
     PrimaryOperand=Convert.Float:Bool(Bool=@TriggerRouteTriggerRouteBool2_0.Value).Float
     AdditionalOperands=UE.Multiply:Float() {also {defaults {AdditionalOperands=-2, PrimaryOperand=180}}}.Out
    ).Out
   )
   TriggerOnValueChange_0 = UE.Trigger On Value Change:Float(Value=@n1.Out)
   n4 = TriggerCompare.Clamp:Float(A=@n1.Out, Compare=@TriggerOnValueChange_0.Trigger) {also {defaults {B=-180, Type=2}}}
   TriggerRouteTriggerRouteBool2_3 = "TriggerRoute.Trigger Route (Bool, 2):Bool"(Set 1=@n4.False, Set 0=@n4.True) {also {defaults {Value 0="<MetaSoundBuilderResult.SUCCEEDED: 0>"}}}
   n5 = UE.Multiply:Float(
    PrimaryOperand=Convert.Float:Bool(Bool=@TriggerRouteTriggerRouteBool2_3.Value).Float
    AdditionalOperands=UE.Multiply:Float() {also {defaults {AdditionalOperands=2, PrimaryOperand=180}}}.Out
   )
   CrossfadeTriggerRouteFloat5_1 = "Crossfade.Trigger Route (Float, 5):Float"(Crossfade Value=MapRange.MapRange:Float(In=UE.Add:Float(PrimaryOperand=@n3.Out, AdditionalOperands=@n5.Out).Out) {also {defaults {In Range A=-180, In Range B=180, Out Range B=4}}}.Out Value) {also {defaults {In 0=0, In 1=-1, In 2=0, In 3=1, In 4=0}}}
   StereoPanner_0 = UE.Stereo Panner(In=^In R, Pan Amount=@CrossfadeTriggerRouteFloat5_1.Out)
   n6 = UE.Subtract:Float(PrimaryOperand=^Azimuth, AdditionalOperands=^StereoSpread)
   TriggerOnValueChange_2 = UE.Trigger On Value Change:Float(Value=@n6.Out)
   n7 = TriggerCompare.Clamp:Float(A=@n6.Out, Compare=@TriggerOnValueChange_2.Trigger) {also {defaults {B=180, Type=3}}}
   TriggerRouteTriggerRouteBool2_2 = "TriggerRoute.Trigger Route (Bool, 2):Bool"(Set 1=@n7.False, Set 0=@n7.True) {also {defaults {Value 0="<MetaSoundBuilderResult.SUCCEEDED: 0>"}}}
   n8 = UE.Add:Float(
    AdditionalOperands=UE.Multiply:Float(
     PrimaryOperand=Convert.Float:Bool(Bool=@TriggerRouteTriggerRouteBool2_2.Value).Float
     AdditionalOperands=UE.Multiply:Float() {also {defaults {AdditionalOperands=-2, PrimaryOperand=180}}}.Out
    ).Out
    PrimaryOperand=@n6.Out
   )
   TriggerOnValueChange_3 = UE.Trigger On Value Change:Float(Value=@n6.Out)
   n9 = TriggerCompare.Clamp:Float(A=@n6.Out, Compare=@TriggerOnValueChange_3.Trigger) {also {defaults {B=-180, Type=2}}}
   TriggerRouteTriggerRouteBool2_1 = "TriggerRoute.Trigger Route (Bool, 2):Bool"(Set 1=@n9.False, Set 0=@n9.True) {also {defaults {Value 0="<MetaSoundBuilderResult.SUCCEEDED: 0>"}}}
   n10 = UE.Multiply:Float(
    PrimaryOperand=Convert.Float:Bool(Bool=@TriggerRouteTriggerRouteBool2_1.Value).Float
    AdditionalOperands=UE.Multiply:Float() {also {defaults {AdditionalOperands=2, PrimaryOperand=180}}}.Out
   )
   CrossfadeTriggerRouteFloat5_0 = "Crossfade.Trigger Route (Float, 5):Float"(Crossfade Value=MapRange.MapRange:Float(In=UE.Add:Float(PrimaryOperand=@n8.Out, AdditionalOperands=@n10.Out).Out) {also {defaults {In Range A=-180, In Range B=180, Out Range B=4}}}.Out Value) {also {defaults {In 0=0, In 1=-1, In 2=0, In 3=1, In 4=0}}}
   StereoPanner_1 = UE.Stereo Panner(In=^In L, Pan Amount=@CrossfadeTriggerRouteFloat5_0.Out)
   AudioMixerAudioMixerStereo2_0 = "AudioMixer.Audio Mixer (Stereo, 2)"(In 1 L=@StereoPanner_0.Out Left, In 1 R=@StereoPanner_0.Out Right, In 0 L=@StereoPanner_1.Out Left, In 0 R=@StereoPanner_1.Out Right)
  ]
  links [{dst="@Out L", src=AudioMixerAudioMixerStereo2_0.Out L}, {dst="@Out R", src=AudioMixerAudioMixerStereo2_0.Out R}]
 }]
