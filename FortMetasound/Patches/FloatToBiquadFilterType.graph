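# MetaSound patch that turns a Float input into a BiquadFilterType enum, so a filter type can be chosen by a numeric value.
# It clamps the float with an upper limit of 9, converts it to Int32, then converts that integer to the enum output.
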MetaSoundPatch /FortMetasound/Patches/FloatToBiquadFilterType
inputs [Input {type=Float}]
outputs ["Enum:BiquadFilterType" {type=Enum:BiquadFilterType}]
pages [Default {
  input_defaults {Input=0}
  nodes [ConvertEnum_0 = Convert.Enum:BiquadFilterType:Int32(Int32=Convert.Int32:Float(Float=Clamp.Clamp:Float(In=^Input) {also {defaults {Max=9}}}.Value).Int32)]
  links [{dst="@Enum:BiquadFilterType", src=ConvertEnum_0.Enum:BiquadFilterType}]
 }]
