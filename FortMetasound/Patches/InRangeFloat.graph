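# MetaSound patch that outputs a Bool indicating whether a float Input lies within a Min to Max range.
# A value-change trigger on Input fires two trigger compares, one against Max and one against Min, each routed to a 1 or 0 float.
# The two floats are multiplied as a logical AND and converted to Bool.
# The Tolerance input is declared but not connected.
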MetaSoundPatch /FortMetasound/Patches/InRangeFloat
inputs [
 Input {type=Float}
 Max {type=Float}
 Min {type=Float}
 Tolerance {type=Float}
]
interfaces [MetaSound:1.0]
outputs [Result {type=Bool}]
pages [Default {
  input_defaults {Input=0, Max=0, Min=0, Tolerance=0.01}
  nodes [
   TriggerOnValueChange_0 = UE.Trigger On Value Change:Float(Value=^Input)
   n1 = TriggerCompare.Clamp:Float(A=^Input, B=^Max, Compare=@TriggerOnValueChange_0.Trigger) {also {defaults {Type=4}}}
   TriggerRouteTriggerRouteFloat2_0 = "TriggerRoute.Trigger Route (Float, 2):Float"(Set 1=@n1.FALSE, Set 0=@n1.TRUE) {also {defaults {Value 0=1, Value 1=0}}}
   n2 = TriggerCompare.Clamp:Float(A=^Input, B=^Min, Compare=@TriggerOnValueChange_0.Trigger) {also {defaults {Type=5}}}
   TriggerRouteTriggerRouteFloat2_1 = "TriggerRoute.Trigger Route (Float, 2):Float"(Set 1=@n2.FALSE, Set 0=@n2.TRUE) {also {defaults {Value 0=1}}}
   ConvertBool_0 = Convert.Bool:Float(Float=UE.Multiply:Float(AdditionalOperands=@TriggerRouteTriggerRouteFloat2_0.Value, PrimaryOperand=@TriggerRouteTriggerRouteFloat2_1.Value).Out)
  ]
  links [{dst="@Result", src=ConvertBool_0.Bool}]
 }]
