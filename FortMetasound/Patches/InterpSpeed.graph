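# MetaSound patch that smooths a float input with separate interpolation settings for rising and falling values.
# A variable with deferred accessors holds the previous input as the threshold for two edge triggers, which route the rising or falling setting.
# InterpTo glides toward the input using that setting converted to time, so the "speed" inputs act as durations.
# It outputs both the smoothed value and the selected setting.
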MetaSoundPatch /FortMetasound/Patches/InterpSpeed
inputs [
 FallingInterpSpeed {type=Float}
 FloatInput {type=Float}
 RisingInterpSpeed {type=Float}
]
interfaces [MetaSound:1.0]
outputs [
 FloatOutput {type=Float}
 InterpValue {type=Float}
]
pages [Default {
  input_defaults {FallingInterpSpeed=0.2, FloatInput=0.82, RisingInterpSpeed=5}
  nodes [
   n1 = InitVariable.Float()
   n2 = VariableDeferredAccessor.Float(Variable=@n1.Variable)
   TriggerRouteTriggerRouteFloat2_0 = "TriggerRoute.Trigger Route (Float, 2):Float"(
    Value 1=^FallingInterpSpeed
    Value 0=^RisingInterpSpeed
    Set 1=UE.TriggerOnThreshold:Float(In=^FloatInput, Threshold=@n2.Value) {also {defaults {Type=1}}}.Out
    Set 0=UE.TriggerOnThreshold:Float(In=^FloatInput, Threshold=VariableDeferredAccessor.Float(Variable=@n1.Variable).Value).Out
   )
   InterpTo_0 = UE.InterpTo:Audio(Target=^FloatInput, Interp Time=UE.ConversionFloatToTime(In=@TriggerRouteTriggerRouteFloat2_0.Value).Out)
   VariableMutator.Float(Value=^FloatInput, Variable=@n2.Variable)
  ]
  links [{dst="@FloatOutput", src=InterpTo_0.Value}, {dst="@InterpValue", src=TriggerRouteTriggerRouteFloat2_0.Value}]
 }]
