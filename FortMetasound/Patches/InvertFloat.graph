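# InvertFloat is a MetaSound patch that inverts a float by computing one minus the input.
# A single float Subtract node holds 1 as its primary operand, subtracts FloatIn and feeds FloatOut.
# Values in the 0 to 1 range map to 1 to 0, and no clamping is applied.
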MetaSoundPatch /FortMetasound/Patches/InvertFloat
inputs [FloatIn {type=Float}]
interfaces [MetaSound:1.0]
outputs [FloatOut {type=Float}]
pages [Default {
  input_defaults {FloatIn=0}
  nodes [Subtract_0 = UE.Subtract:Float(AdditionalOperands=^FloatIn) {also {defaults {PrimaryOperand=1}}}]
  links [{dst="@FloatOut", src=Subtract_0.Out}]
 }]
