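# MetaSound patch that tests whether a WaveAsset input is valid and fires either an Is Valid or Is Not Valid trigger.
# On Trigger In, it converts the wave's duration from time to float and feeds it to a Trigger Compare node left at its default comparison.
# A true result fires Is Not Valid and a false one fires Is Valid, so duration stands in for a direct validity check.
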MetaSoundPatch /FortMetasound/Patches/IsWaveAssetValid
inputs [
 Trigger In {type=Trigger}
 WaveAsset {type=WaveAsset}
]
interfaces [MetaSound:1.0]
outputs [
 Is Not Valid {type=Trigger}
 Is Valid {type=Trigger}
]
pages [Default {
  input_defaults {Trigger In=null, WaveAsset=null}
  nodes [
   GetWaveDuration_0 = "UE.Get Wave Duration: "(Wave=^WaveAsset)
   TriggerCompareClamp_0 = TriggerCompare.Clamp:Float(Compare=^Trigger In, A=UE.ConversionTimeToFloat(In=@GetWaveDuration_0.Duration).Out)
  ]
  links [{dst="@Is Valid", src=TriggerCompareClamp_0.FALSE}, {dst="@Is Not Valid", src=TriggerCompareClamp_0.TRUE}]
 }]
