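# MetaSound patch that scales a float by a mix value chosen from the sound's FortAudioAffiliation, giving separate Local, Enemy and Non-Enemy levels.
# The affiliation enum converts to a float that drives a four-input crossfade, with index 0 as Local, 2 as Enemy, and both 1 and 3 as Non-Enemy.
# The selected value multiplies In Value to produce Out, which passes the input through unchanged at default settings.
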MetaSoundPatch /FortMetasound/Patches/MSP_AffiliationMixFloat
inputs [
 Affiliation {type=Enum:FortAudioAffiliation}
 Enemy {type=Float}
 In Value {type=Float}
 Local {type=Float}
 Non-Enemy {type=Float}
]
outputs [Out {type=Float}]
pages [Default {
  input_defaults {Affiliation=0, Enemy=1, In Value=1, Local=1, Non-Enemy=1}
  nodes [
   CrossfadeTriggerRouteFloat4_0 = "Crossfade.Trigger Route (Float, 4):Float"(
    In 2=^Enemy
    In 0=^Local
    In 1=^Non-Enemy
    In 3=^Non-Enemy
    Crossfade Value=Convert.Float:Int32(Int32=Convert.Int32:Enum:FortAudioAffiliation("Enum:FortAudioAffiliation"=^Affiliation).Int32).Float
   )
   Multiply_0 = UE.Multiply:Float(AdditionalOperands=^In Value, PrimaryOperand=@CrossfadeTriggerRouteFloat4_0.Out)
  ]
  links [{dst="@Out", src=Multiply_0.Out}]
 }]
