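# MetaSound patch that turns an integer Affiliation into Local, Friendly or Enemy triggers for audio logic.
# OnPlay and any change in Affiliation merge through a Trigger Any into three parallel Trigger Compare nodes.
# These test Affiliation against 0, 1 and 2, and their true outputs drive Local, Friendly and Enemy respectively.
# An audio Trigger Route node is present but unconnected.
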MetaSoundPatch /FortMetasound/Patches/MSP_Audio_AffiliationSwitch
inputs [
 Affiliation {type=Int32}
 OnPlay {type=Trigger}
]
outputs [
 Enemy {type=Trigger}
 Friendly {type=Trigger}
 Local {type=Trigger}
]
pages [Default {
  input_defaults {Affiliation=0, OnPlay=null}
  nodes [
   TriggerOnValueChange_0 = UE.Trigger On Value Change:Int32(Value=^Affiliation)
   TriggerAnyTriggerAny2_0 = "TriggerAny.Trigger Any (2)"(In 0=^OnPlay, In 1=@TriggerOnValueChange_0.Trigger)
   TriggerCompareClamp_0 = TriggerCompare.Clamp:Int32(A=^Affiliation, Compare=@TriggerAnyTriggerAny2_0.out) {also {defaults {B=2}}}
   TriggerCompareClamp_1 = TriggerCompare.Clamp:Int32(A=^Affiliation, Compare=@TriggerAnyTriggerAny2_0.out) {also {defaults {B=1}}}
   TriggerCompareClamp_2 = TriggerCompare.Clamp:Int32(A=^Affiliation, Compare=@TriggerAnyTriggerAny2_0.out) {also {defaults {B=0}}}
   TriggerRouteTriggerRouteAudio3_0 = "TriggerRoute.Trigger Route (Audio, 3):Audio"()
  ]
  links [
   {dst="@Enemy", src=TriggerCompareClamp_0.TRUE}
   {dst="@Friendly", src=TriggerCompareClamp_1.TRUE}
   {dst="@Local", src=TriggerCompareClamp_2.TRUE}
  ]
 }]
