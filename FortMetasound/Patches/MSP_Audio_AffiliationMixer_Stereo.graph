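# Stereo MetaSound patch that applies separate gain and filter settings to audio for Local, Friendly or Enemy affiliation.
# On play, MSP_Audio_AffiliationSwitch fires the matching trigger, which routes that affiliation's dB adjustment, highpass and lowpass values through variables.
# Each channel runs through one-pole high-pass and low-pass filters, then a stereo mixer applies the gain converted from decibels.
# Default settings are 0 dB with wide-open cutoffs.
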MetaSoundPatch /FortMetasound/Patches/MSP_Audio_AffiliationMixer_Stereo
inputs [
 Affiliation {type=Int32}
 Enemy.DbAdjustment {type=Float}
 Enemy.Highpass {type=Float}
 Enemy.Lowpass {type=Float}
 Friendly.DbAdjustment {type=Float}
 Friendly.Highpass {type=Float}
 Friendly.Lowpass {type=Float}
 In L {type=Audio}
 In R {type=Audio}
 Local.DbAdjustment {type=Float}
 Local.Highpass {type=Float}
 Local.Lowpass {type=Float}
 OnPlay {type=Trigger}
]
outputs [
 Out L {type=Audio}
 Out R {type=Audio}
]
pages [Default {
  input_defaults {
   Affiliation=0
   Enemy.DbAdjustment=0
   Enemy.Highpass=0
   Enemy.Lowpass=20000
   Friendly.DbAdjustment=0
   Friendly.Highpass=0
   Friendly.Lowpass=20000
   In L=null
   In R=null
   Local.DbAdjustment=0
   Local.Highpass=0
   Local.Lowpass=20000
   OnPlay=null
  }
  nodes [
   n1 = InitVariable.Float()
   n2 = &/MSP_Audio_AffiliationSwitch(Affiliation=^Affiliation, OnPlay=^OnPlay)
   n3 = VariableAccessor.Trigger(Variable=VariableMutator.Trigger(Variable=InitVariable.Trigger().Variable, Value=@n2.Friendly).Variable)
   n4 = VariableAccessor.Trigger(Variable=@n3.Variable)
   n5 = VariableAccessor.Trigger(Variable=VariableMutator.Trigger(Variable=InitVariable.Trigger().Variable, Value=@n2.Enemy).Variable)
   n6 = VariableAccessor.Trigger(Variable=@n5.Variable)
   n7 = VariableAccessor.Trigger(Variable=VariableMutator.Trigger(Variable=InitVariable.Trigger().Variable, Value=@n2.Local).Variable)
   n8 = VariableAccessor.Trigger(Variable=@n7.Variable)
   TriggerRouteTriggerRouteFloat3_2 = "TriggerRoute.Trigger Route (Float, 3):Float"(
    Value 2=^Enemy.DbAdjustment
    Value 1=^Friendly.DbAdjustment
    Value 0=^Local.DbAdjustment
    Set 1=VariableAccessor.Trigger(Variable=@n4.Variable).Value
    Set 2=VariableAccessor.Trigger(Variable=@n6.Variable).Value
    Set 0=VariableAccessor.Trigger(Variable=@n8.Variable).Value
   )
   DecibelstoLinearGain_0 = UE.Decibels to Linear Gain:Float(Decibels=VariableAccessor.Float(Variable=VariableMutator.Float(Variable=@n1.Variable, Value=@TriggerRouteTriggerRouteFloat3_2.Value).Variable).Value)
   TriggerRouteTriggerRouteFloat3_0 = "TriggerRoute.Trigger Route (Float, 3):Float"(Value 2=^Enemy.Highpass, Value 1=^Friendly.Highpass, Value 0=^Local.Highpass, Set 1=@n3.Value, Set 2=@n5.Value, Set 0=@n7.Value)
   n9 = VariableAccessor.Float(Variable=VariableMutator.Float(Variable=InitVariable.Float().Variable, Value=@TriggerRouteTriggerRouteFloat3_0.Value).Variable)
   OnePoleHighPassFilter_1 = UE.One-Pole High Pass Filter:Audio(In=^In R, Cutoff Frequency=@n9.Value)
   TriggerRouteTriggerRouteFloat3_1 = "TriggerRoute.Trigger Route (Float, 3):Float"(Value 2=^Enemy.Lowpass, Value 1=^Friendly.Lowpass, Value 0=^Local.Lowpass, Set 0=@n8.Value, Set 1=@n4.Value, Set 2=@n6.Value)
   n10 = VariableAccessor.Float(Variable=VariableMutator.Float(Variable=InitVariable.Float().Variable, Value=@TriggerRouteTriggerRouteFloat3_1.Value).Variable)
   OnePoleLowPassFilter_0 = UE.One-Pole Low Pass Filter:Audio(In=@OnePoleHighPassFilter_1.out, Cutoff Frequency=@n10.Value)
   OnePoleHighPassFilter_0 = UE.One-Pole High Pass Filter:Audio(In=^In L, Cutoff Frequency=@n9.Value)
   OnePoleLowPassFilter_1 = UE.One-Pole Low Pass Filter:Audio(In=@OnePoleHighPassFilter_0.out, Cutoff Frequency=@n10.Value)
   AudioMixerAudioMixerStereo2_0 = "AudioMixer.Audio Mixer (Stereo, 2)"(Gain 0=@DecibelstoLinearGain_0.Linear Gain, In 0 R=@OnePoleLowPassFilter_0.out, In 0 L=@OnePoleLowPassFilter_1.out)
  ]
  links [{dst="@Out L", src=AudioMixerAudioMixerStereo2_0.Out L}, {dst="@Out R", src=AudioMixerAudioMixerStereo2_0.Out R}]
 }]
