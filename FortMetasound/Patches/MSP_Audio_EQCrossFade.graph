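# MetaSound patch that morphs EQ and level on a stereo signal as a float input moves between Min and Max values.
# The input is smoothed over Interp Time, then range-mapped to high-pass cutoff, low-pass cutoff and gain.
# Each channel runs through one-pole high-pass and low-pass filters into a stereo mixer that applies the gain.
# The dB endpoints convert to linear gain before mapping, so level interpolates linearly in amplitude.
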MetaSoundPatch /FortMetasound/Patches/MSP_Audio_EQCrossFade
inputs [
 In L {type=Audio}
 In R {type=Audio}
 Input {type=Float}
 Interp Time {type=Time}
 Max.DbAdjustment {type=Float}
 Max.HPF {type=Float}
 Max.LPF {type=Float}
 Max.Value {type=Float}
 Min.DbAdjustment {type=Float}
 Min.HPF {type=Float}
 Min.LPF {type=Float}
 Min.Value {type=Float}
]
outputs [
 Out L {type=Audio}
 Out R {type=Audio}
]
pages [Default {
  input_defaults {
   In L=null
   In R=null
   Input=0
   Interp Time=0
   Max.DbAdjustment=0
   Max.HPF=0
   Max.LPF=20000
   Max.Value=1
   Min.DbAdjustment=0
   Min.HPF=0
   Min.LPF=20000
   Min.Value=0
  }
  nodes [
   n1 = VariableAccessor.Float(Variable=VariableMutator.Float(Variable=InitVariable.Float().Variable, Value=UE.InterpTo:Audio(Target=^Input, Interp Time=^Interp Time).Value).Variable)
   n2 = VariableAccessor.Float(Variable=VariableMutator.Float(
     Variable=InitVariable.Float().Variable
     Value=MapRange.MapRange:Float(Out Range B=^Max.HPF, In Range B=^Max.Value, Out Range A=^Min.HPF, In Range A=^Min.Value, In=@n1.Value).Out Value
    ).Variable)
   OnePoleHighPassFilter_1 = UE.One-Pole High Pass Filter:Audio(In=^In R, Cutoff Frequency=@n2.Value)
   n3 = VariableAccessor.Float(Variable=@n1.Variable)
   n4 = VariableAccessor.Float(Variable=VariableMutator.Float(
     Variable=InitVariable.Float().Variable
     Value=MapRange.MapRange:Float(Out Range B=^Max.LPF, In Range B=^Max.Value, Out Range A=^Min.LPF, In Range A=^Min.Value, In=@n3.Value).Out Value
    ).Variable)
   OnePoleLowPassFilter_0 = UE.One-Pole Low Pass Filter:Audio(In=@OnePoleHighPassFilter_1.out, Cutoff Frequency=@n4.Value)
   OnePoleHighPassFilter_0 = UE.One-Pole High Pass Filter:Audio(In=^In L, Cutoff Frequency=@n2.Value)
   OnePoleLowPassFilter_1 = UE.One-Pole Low Pass Filter:Audio(In=@OnePoleHighPassFilter_0.out, Cutoff Frequency=@n4.Value)
   n5 = InitVariable.Float()
   DecibelstoLinearGain_0 = UE.Decibels to Linear Gain:Float(Decibels=^Min.DbAdjustment)
   DecibelstoLinearGain_1 = UE.Decibels to Linear Gain:Float(Decibels=^Max.DbAdjustment)
   n6 = MapRange.MapRange:Float(
    In Range B=^Max.Value
    In Range A=^Min.Value
    Out Range A=@DecibelstoLinearGain_0.Linear Gain
    Out Range B=@DecibelstoLinearGain_1.Linear Gain
    In=VariableAccessor.Float(Variable=@n3.Variable).Value
   )
   AudioMixerAudioMixerStereo2_0 = "AudioMixer.Audio Mixer (Stereo, 2)"(
    In 0 R=@OnePoleLowPassFilter_0.out
    In 0 L=@OnePoleLowPassFilter_1.out
    Gain 0=VariableAccessor.Float(Variable=VariableMutator.Float(Variable=@n5.Variable, Value=@n6.Out Value).Variable).Value
   )
  ]
  links [{dst="@Out L", src=AudioMixerAudioMixerStereo2_0.Out L}, {dst="@Out R", src=AudioMixerAudioMixerStereo2_0.Out R}]
 }]
