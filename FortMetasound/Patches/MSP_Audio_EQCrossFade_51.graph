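# MetaSound patch that applies an EQ crossfade to 5.1 audio, driven by a single float Input.
# It runs three instances of the stereo MSP_Audio_EQCrossFade patch on the L/R, C/LF and SL/SR pairs.
# All three share the Input, the interp time and the Min/Max value, dB adjustment, high-pass and low-pass settings, so the LF channel gets the same treatment as the others.
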MetaSoundPatch /FortMetasound/Patches/MSP_Audio_EQCrossFade_51
inputs [
 In C {type=Audio}
 In L {type=Audio}
 In LF {type=Audio}
 In R {type=Audio}
 In SL {type=Audio}
 In SR {type=Audio}
 Input {type=Float}
 Interp Time {type=Time}
 Max.DbAdjustment {type=Float}
 Max.HPF {type=Float}
 Max.LPF {type=Float}
 Max.Value {type=Float}
 Min.DbAdjustment {type=Float}
 Min.HPF {type=Float}
 Min.LPF {type=Float}
 Min.Value {type=Float}
]
outputs [
 Out C {type=Audio}
 Out L {type=Audio}
 Out LF {type=Audio}
 Out R {type=Audio}
 Out SL {type=Audio}
 Out SR {type=Audio}
]
pages [Default {
  input_defaults {
   In C=null
   In L=null
   In LF=null
   In R=null
   In SL=null
   In SR=null
   Input=0
   Interp Time=0
   Max.DbAdjustment=0
   Max.HPF=0
   Max.LPF=20000
   Max.Value=1
   Min.DbAdjustment=0
   Min.HPF=0
   Min.LPF=20000
   Min.Value=0
  }
  nodes [
   InitVariable.Float()
   InitVariable.Float()
   InitVariable.Float()
   n1 = VariableAccessor.Float(Variable=VariableMutator.Float(Value=^Input, Variable=InitVariable.Float().Variable).Variable)
   n2 = VariableAccessor.Float(Variable=@n1.Variable)
   MSP_Audio_EQCrossFade_0 = &/MSP_Audio_EQCrossFade(
    In L=^In SL
    In R=^In SR
    Interp Time=^Interp Time
    Max.DbAdjustment=^Max.DbAdjustment
    Max.HPF=^Max.HPF
    Max.LPF=^Max.LPF
    Max.Value=^Max.Value
    Min.DbAdjustment=^Min.DbAdjustment
    Min.HPF=^Min.HPF
    Min.LPF=^Min.LPF
    Min.Value=^Min.Value
    Input=VariableAccessor.Float(Variable=@n2.Variable).Value
   )
   MSP_Audio_EQCrossFade_1 = &MSP_Audio_EQCrossFade(
    In L=^In C
    In R=^In LF
    Interp Time=^Interp Time
    Max.DbAdjustment=^Max.DbAdjustment
    Max.HPF=^Max.HPF
    Max.LPF=^Max.LPF
    Max.Value=^Max.Value
    Min.DbAdjustment=^Min.DbAdjustment
    Min.HPF=^Min.HPF
    Min.LPF=^Min.LPF
    Min.Value=^Min.Value
    Input=@n2.Value
   )
   MSP_Audio_EQCrossFade_2 = &MSP_Audio_EQCrossFade(
    In L=^In L
    In R=^In R
    Interp Time=^Interp Time
    Max.DbAdjustment=^Max.DbAdjustment
    Max.HPF=^Max.HPF
    Max.LPF=^Max.LPF
    Max.Value=^Max.Value
    Min.DbAdjustment=^Min.DbAdjustment
    Min.HPF=^Min.HPF
    Min.LPF=^Min.LPF
    Min.Value=^Min.Value
    Input=@n1.Value
   )
  ]
  links [
   {dst="@Out SL", src=MSP_Audio_EQCrossFade_0.Out L}
   {dst="@Out SR", src=MSP_Audio_EQCrossFade_0.Out R}
   {dst="@Out C", src=MSP_Audio_EQCrossFade_1.Out L}
   {dst="@Out LF", src=MSP_Audio_EQCrossFade_1.Out R}
   {dst="@Out L", src=MSP_Audio_EQCrossFade_2.Out L}
   {dst="@Out R", src=MSP_Audio_EQCrossFade_2.Out R}
  ]
 }]
