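# MetaSound patch that plays a randomly chosen stereo wave from an input array, with looping, pitch shift, start time and decibel gain controls.
# Play Trigger advances an Array Random Get, optionally with shared random state, whose pick feeds and triggers a stereo Wave Player.
# Both channels are scaled by the linear-converted gain.
# On Play passes Play Trigger straight through rather than reporting from the player.
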MetaSoundPatch /FortMetasound/Patches/MSP_Audio_WavePlayer_Basic_Stereo
inputs [
 Enable Shared Random State {type=Bool}
 Gain in Decibels {type=Float}
 Is Looping? {type=Bool}
 Pitch Shift {type=Float}
 Play Trigger {type=Trigger}
 Start Time {type=Time}
 Wave Asset {type=WaveAsset:Array}
]
outputs [
 On Play {type=Trigger}
 Out L {type=Audio}
 Out R {type=Audio}
]
pages [Default {
  input_defaults {Enable Shared Random State=null, Gain in Decibels=0, Is Looping?=null, Pitch Shift=0, Play Trigger=null, Start Time=0, Wave Asset []}
  nodes [
   DecibelstoLinearGain_0 = UE.Decibels to Linear Gain:Float(Decibels=^Gain in Decibels)
   ArrayRandomGet_0 = Array.Random Get:WaveAsset:Array(Enable Shared State=^Enable Shared Random State, Next=^Play Trigger, In Array=^Wave Asset) {also {version=~}}
   WavePlayer_0 = UE.Wave Player:Stereo(Loop=^Is Looping?, Pitch Shift=^Pitch Shift, Start Time=^Start Time, Play=@ArrayRandomGet_0.On Next, Wave Asset=@ArrayRandomGet_0.Value)
   Multiply_0 = UE.Multiply:Audio by Float(AdditionalOperands=@DecibelstoLinearGain_0.Linear Gain, PrimaryOperand=@WavePlayer_0.Out Right)
   Multiply_1 = UE.Multiply:Audio by Float(AdditionalOperands=@DecibelstoLinearGain_0.Linear Gain, PrimaryOperand=@WavePlayer_0.Out Left)
  ]
  links [{dst="@On Play", src="@Play Trigger"}, {dst="@Out R", src=Multiply_0.Out}, {dst="@Out L", src=Multiply_1.Out}]
 }]
