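# A reusable MetaSound patch that plays a random stereo wave from an array with randomized pitch, delay, start time and volume.
# Each trigger chains pitch roll, random delay, array pick, gain roll and start-time roll before starting the Wave Player.
# Global offsets add to pitch and start time, while fixed and random gains, converted from decibels, multiply both channels.
# On Play fires immediately, before the delay.
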MetaSoundPatch /FortMetasound/Patches/MSP_Audio_WavePlayer_Deluxe_Stereo
inputs [
 "Delay Max (Seconds)" {type=Time}
 "Delay Min (Seconds)" {type=Time}
 Enable Shared Random State {type=Bool}
 Is Looping? {type=Bool}
 Max Random Decibels {type=Float}
 Min Random Decibels {type=Float}
 "Pitch Shift Global (Semitones)" {type=Float}
 Play Trigger {type=Trigger}
 "Random Pitch Max (Semitones)" {type=Float}
 "Random Pitch Min (Semitones)" {type=Float}
 Start Time Global Offset {type=Time}
 Start Time Max Random {type=Time}
 "Start Time Min Random " {type=Time}
 Wave Asset {type=WaveAsset:Array}
 WavePlayer Gain dB {type=Float}
]
outputs [
 On Finished {type=Trigger}
 On Play {type=Trigger}
 Out L {type=Audio}
 Out R {type=Audio}
]
pages [Default {
  input_defaults {
   "Delay Max (Seconds)"=0
   "Delay Min (Seconds)"=0
   Enable Shared Random State=null
   Is Looping?=null
   Max Random Decibels=0
   Min Random Decibels=0
   "Pitch Shift Global (Semitones)"=0
   Play Trigger=null
   "Random Pitch Max (Semitones)"=1
   "Random Pitch Min (Semitones)"=-1
   Start Time Global Offset=0
   Start Time Max Random=0
   "Start Time Min Random "=0
   WavePlayer Gain dB=0
   Wave Asset []
  }
  nodes [
   DecibelstoLinearGain_2 = UE.Decibels to Linear Gain:Float(Decibels=^WavePlayer Gain dB)
   n1 = UE.RandomFloat(Next=^Play Trigger, Max=^"Random Pitch Max (Semitones)", Min=^"Random Pitch Min (Semitones)")
   n2 = UE.RandomTime(Max=^"Delay Max (Seconds)", Min=^"Delay Min (Seconds)", Next=@n1.On Next)
   TriggerDelay_0 = UE.Trigger Delay(In=@n2.On Next, Delay Time=@n2.Value)
   ArrayRandomGet_0 = Array.Random Get:WaveAsset:Array(Enable Shared State=^Enable Shared Random State, In Array=^Wave Asset, Next=@TriggerDelay_0.Out)
   DecibelstoLinearGain_0 = UE.Decibels to Linear Gain:Float(Decibels=^Min Random Decibels)
   DecibelstoLinearGain_1 = UE.Decibels to Linear Gain:Float(Decibels=^Max Random Decibels)
   n3 = UE.RandomFloat(Next=@ArrayRandomGet_0.On Next, Min=@DecibelstoLinearGain_0.Linear Gain, Max=@DecibelstoLinearGain_1.Linear Gain)
   n4 = UE.RandomTime(Max=^Start Time Max Random, Min=^"Start Time Min Random ", Next=@n3.On Next)
   WavePlayer_0 = UE.Wave Player:Stereo(
    Loop=^Is Looping?
    Pitch Shift=UE.Add:Float(AdditionalOperands=^"Pitch Shift Global (Semitones)", PrimaryOperand=@n1.Value).Out
    Start Time=UE.Add:Time(AdditionalOperands=^Start Time Global Offset, PrimaryOperand=@n4.Value).Out
    Wave Asset=@ArrayRandomGet_0.Value
    Play=@n4.On Next
   ) {also {version [1,1]}}
   Multiply_0 = UE.Multiply:Audio by Float(
    PrimaryOperand=UE.Multiply:Audio by Float(AdditionalOperands=@DecibelstoLinearGain_2.Linear Gain, PrimaryOperand=@WavePlayer_0.Out Right).Out
    AdditionalOperands=@n3.Value
   )
   DecibelstoLinearGain_3 = UE.Decibels to Linear Gain:Float(Decibels=^WavePlayer Gain dB)
   Multiply_2 = UE.Multiply:Audio by Float(
    PrimaryOperand=UE.Multiply:Audio by Float(AdditionalOperands=@DecibelstoLinearGain_3.Linear Gain, PrimaryOperand=@WavePlayer_0.Out Left).Out
    AdditionalOperands=@n3.Value
   )
  ]
  links [
   {dst="@On Play", src="@Play Trigger"}
   {dst="@Out R", src=Multiply_0.Out}
   {dst="@Out L", src=Multiply_2.Out}
   {dst="@On Finished", src=WavePlayer_0.On Finished}
  ]
 }]
