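# MSP_AzimuthMapper is a MetaSound patch that turns an azimuth angle into a float by blending values set for Back, Left, Front and Right.
# A MapRange converts Azimuth from -180 to 180 degrees into a crossfade position across a five-input float crossfade.
# Back feeds both end inputs, so the output matches where the angle wraps behind.
# Default direction values give a left-negative, right-positive result, with zero at front and back.
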MetaSoundPatch /FortMetasound/Patches/MSP_AzimuthMapper
inputs [
 Azimuth {type=Float}
 Back {type=Float}
 Front {type=Float}
 Left {type=Float}
 Right {type=Float}
]
interfaces [MetaSound:1.0]
outputs [Out {type=Float}]
pages [Default {
  input_defaults {Azimuth=0, Back=0, Front=0, Left=-1, Right=1}
  nodes [CrossfadeTriggerRouteFloat5_0 = "Crossfade.Trigger Route (Float, 5):Float"(
    In 0=^Back
    In 4=^Back
    In 2=^Front
    In 1=^Left
    In 3=^Right
    Crossfade Value=MapRange.MapRange:Float(In=^Azimuth) {also {defaults {In Range A=-180, In Range B=180, Out Range B=4}}}.Out Value
   )]
  links [{dst="@Out", src=CrossfadeTriggerRouteFloat5_0.Out}]
 }]
