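# MetaSound patch that limits how far a float can move per update, clamping each change to within Tolerance of the previous output.
# OnPlay seeds the raw Value and opens a gate for comparisons triggered by value changes or ForceCalculation.
# Changes within Tolerance pass through, while larger ones step the previous output up or down by Tolerance.
# Deferred variables feed the stored output and triggers back without a graph cycle.
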MetaSoundPatch /FortMetasound/Patches/MSP_ClampFloatChange
inputs [
 ForceCalculation {type=Trigger}
 OnPlay {type=Trigger}
 Tolerance {type=Float}
 Value {type=Float}
]
interfaces [MetaSound:1.0]
outputs [Output {type=Float}]
pages [Default {
  input_defaults {ForceCalculation=null, OnPlay=null, Tolerance=100, Value=0}
  nodes [
   TriggerOnValueChange_1 = UE.Trigger On Value Change:Float(Value=^Value)
   TriggerAnyTriggerAny2_0 = "TriggerAny.Trigger Any (2)"(In 0=^ForceCalculation, In 1=@TriggerOnValueChange_1.Trigger)
   n1 = InitVariable.Float()
   n2 = VariableDeferredAccessor.Float(Variable=@n1.Variable)
   n3 = VariableDeferredAccessor.Float(Variable=@n1.Variable)
   n4 = VariableDeferredAccessor.Float(Variable=@n3.Variable)
   TriggerOnValueChange_0 = UE.Trigger On Value Change:Float(Value=^Value)
   TriggerAnyTriggerAny2_1 = "TriggerAny.Trigger Any (2)"(In 0=^ForceCalculation, In 1=@TriggerOnValueChange_0.Trigger)
   n5 = VariableDeferredAccessor.Trigger(Variable=InitVariable.Trigger().Variable)
   TriggerControl_0 = UE.Trigger Control(Trigger In=@TriggerAnyTriggerAny2_1.Out, Open=@n5.Value)
   n6 = TriggerCompare.Clamp:Float(
    B=^Tolerance
    A=AbsoluteValue.Abs:Float(Input=UE.Subtract:Float(PrimaryOperand=^Value, AdditionalOperands=@n4.Value).Out).Value
    Compare=@TriggerControl_0.Trigger Out
   ) {also {defaults {Type=3}}}
   n7 = TriggerCompare.Clamp:Float(
    A=UE.Subtract:Float(PrimaryOperand=^Value, AdditionalOperands=VariableDeferredAccessor.Float(Variable=@n1.Variable).Value).Out
    Compare=@n6.TRUE
   ) {also {defaults {B=0, Type=3}}}
   TriggerRouteTriggerRouteFloat4_0 = "TriggerRoute.Trigger Route (Float, 4):Float"(
    Set 0=^OnPlay
    Value 0=^Value
    Value 3=^Value
    Value 1=UE.Add:Float(PrimaryOperand=^Tolerance, AdditionalOperands=@n2.Value).Out
    Value 2=UE.Subtract:Float(AdditionalOperands=^Tolerance, PrimaryOperand=@n2.Value).Out
    Set 3=@n6.FALSE
    Set 2=@n7.FALSE
    Set 1=@n7.TRUE
   )
   n8 = VariableDeferredAccessor.Trigger(Variable=InitVariable.Trigger().Variable)
   TriggerRouteTriggerRouteFloat2_0 = "TriggerRoute.Trigger Route (Float, 2):Float"(Set 0=@TriggerAnyTriggerAny2_0.Out, Value 0=@TriggerRouteTriggerRouteFloat4_0.Value, Value 1=@n3.Value, Set 1=@n8.Value)
   VariableMutator.Float(Value=@TriggerRouteTriggerRouteFloat2_0.Value, Variable=@n4.Variable)
   VariableMutator.Trigger(Value=@TriggerRouteTriggerRouteFloat4_0.On Set, Variable=@n5.Variable)
   VariableMutator.Trigger(Value=@TriggerAnyTriggerAny2_0.Out, Variable=@n8.Variable)
  ]
  links [{dst="@Output", src=TriggerRouteTriggerRouteFloat2_0.Value}]
 }]
