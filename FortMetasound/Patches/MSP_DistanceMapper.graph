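# MetaSound patch that maps a Distance input to a float volume, using an inner radius and a falloff span.
# An in-range check up to just below Inner Radius gates a flat Inner Volume term.
# A second comparison on Distance, with no wired threshold, gates a remap from Inner to Final Volume across the falloff span.
# Both checks become float multipliers, and the two gated terms are summed into Output.
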MetaSoundPatch /FortMetasound/Patches/MSP_DistanceMapper
inputs [
 Distance {type=Float}
 Falloff {type=Float}
 Final Volume {type=Float}
 Inner Radius {type=Float}
 Inner Volume {type=Float}
]
interfaces [MetaSound:1.0]
outputs [Output {type=Float}]
pages [Default {
  input_defaults {Distance=0, Falloff=0, Final Volume=0, Inner Radius=0, Inner Volume=0}
  nodes [
   n1 = UE.Subtract:Float(PrimaryOperand=^Inner Radius) {also {defaults {AdditionalOperands=0.01}}}
   n2 = &/InRangeFloat(Input=^Distance, Max=@n1.Out)
   n3 = &/CompareFloat(A=^Distance) {also {defaults {Type=3}}}
   n4 = UE.Add:Float(AdditionalOperands=^Falloff, PrimaryOperand=^Inner Radius)
   Add_0 = UE.Add:Float(
    PrimaryOperand=UE.Multiply:Float(
     PrimaryOperand=Convert.Float:Bool(Bool=@n2.Result).Float
     AdditionalOperands=MapRange.MapRange:Float(In=^Distance, Out Range A=^Inner Volume, Out Range B=^Inner Volume, In Range B=@n1.Out).Out Value
    ).Out
    AdditionalOperands=UE.Multiply:Float(
     PrimaryOperand=Convert.Float:Bool(Bool=@n3.Result).Float
     AdditionalOperands=MapRange.MapRange:Float(In=^Distance, Out Range B=^Final Volume, In Range A=^Inner Radius, Out Range A=^Inner Volume, In Range B=@n4.Out).Out Value
    ).Out
   )
  ]
  links [{dst="@Output", src=Add_0.Out}]
 }]
