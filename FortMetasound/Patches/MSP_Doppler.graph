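# MetaSound patch that simulates a Doppler effect, outputting a semitone pitch modifier from distance and speed inputs.
# On each distance change, it compares against the previous value, held in a deferred variable, to choose whether speed adds to or subtracts from the speed of sound.
# Speed is scaled from centimeters to meters and by a clamped strength multiplier.
# The resulting frequency ratio is smoothed over the interpolation time, then converted to semitones.
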MetaSoundPatch /FortMetasound/Patches/MSP_Doppler
inputs [
 Distance {type=Float}
 InterpTime {type=Time}
 Speed {type=Float}
 StrengthMultiplier {type=Float}
]
outputs [SemitoneModifier {type=Float}]
pages [Default {
  input_defaults {Distance=0, InterpTime=0.25, Speed=0, StrengthMultiplier=1}
  nodes [
   n1 = UE.Divide:Float(AdditionalOperands=Max.Max:Float(A=^StrengthMultiplier) {also {defaults {B=0.01}}}.Value) {also {defaults {PrimaryOperand=100}}}
   TriggerOnValueChange_0 = UE.Trigger On Value Change:Float(Value=^Distance)
   n2 = VariableDeferredAccessor.Float(Variable=InitVariable.Float().Variable)
   n3 = TriggerCompare.Clamp:Float(A=^Distance, Compare=@TriggerOnValueChange_0.Trigger, B=@n2.Value) {also {defaults {Type=4}}}
   TriggerRouteTriggerRouteFloat2_0 = "TriggerRoute.Trigger Route (Float, 2):Float"(Set 1=@n3.False, Set 0=@n3.True) {also {defaults {Value 0=-1, Value 1=1}}}
   n4 = UE.Add:Float(PrimaryOperand=UE.Multiply:Float(
     PrimaryOperand=UE.Divide:Float(PrimaryOperand=^Speed, AdditionalOperands=@n1.Out).Out
     AdditionalOperands=@TriggerRouteTriggerRouteFloat2_0.Value
    ).Out) {also {defaults {AdditionalOperands=340.29}}}
   FrequencyMultipliertoSemitone_0 = UE.Frequency Multiplier to Semitone:Float(Frequency Multiplier=UE.InterpTo:Audio(Interp Time=^InterpTime, Target=UE.Divide:Float(AdditionalOperands=@n4.Out) {also {defaults {PrimaryOperand=340.29}}}.Out).Value)
   VariableMutator.Float(Value=^Distance, Variable=@n2.Variable)
  ]
  links [{dst="@SemitoneModifier", src=FrequencyMultipliertoSemitone_0.Semitones}]
 }]
