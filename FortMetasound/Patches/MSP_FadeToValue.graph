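# MetaSound patch that fades a float from an initial value to a target over a set time, or back again on a reverse trigger.
# Two trigger routes swap the start and end values by direction, and an ADSR envelope's attack drives a MapRange between them to produce the output.
# When the attack ends, a stored direction bool selects the fade-done or reverse-done trigger; a separate trigger fires at attack start.
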MetaSoundPatch /FortMetasound/Patches/MSP_FadeToValue
inputs [
 Fade {type=Trigger}
 FadeTime {type=Time}
 Initial Value {type=Float}
 ReverseFade {type=Trigger}
 TargetValue {type=Float}
]
outputs [
 On Trigger {type=Trigger}
 OnDoneFade {type=Trigger}
 OnDoneReverseFade {type=Trigger}
 Out Value {type=Float}
]
pages [Default {
  input_defaults {Fade=null, FadeTime=0.1, Initial Value=null, ReverseFade=null, TargetValue=1}
  nodes [
   InitVariable.Trigger()
   TriggerRouteTriggerRouteFloat2_0 = "TriggerRoute.Trigger Route (Float, 2):Float"(Set 0=^Fade, Value 0=^Initial Value, Set 1=^ReverseFade, Value 1=^TargetValue)
   ADSREnvelopeADSREnvelope_0 = ADSR Envelope.ADSR Envelope:Float(Attack Time=^FadeTime, Trigger Attack=@TriggerRouteTriggerRouteFloat2_0.On Set) {also {defaults {Decay Time=0, Hard Reset="<MetaSoundBuilderResult.SUCCEEDED: 0>", Release Time=0, Sustain Level=1}}}
   TriggerRouteTriggerRouteFloat2_1 = "TriggerRoute.Trigger Route (Float, 2):Float"(Set 0=^Fade, Value 1=^Initial Value, Set 1=^ReverseFade, Value 0=^TargetValue)
   MapRangeMapRange_0 = MapRange.MapRange:Float(
    In=@ADSREnvelopeADSREnvelope_0.Out Envelope
    Out Range B=UE.Reroute(Value=@TriggerRouteTriggerRouteFloat2_1.Value).Value
    Out Range A=UE.Reroute(Value=@TriggerRouteTriggerRouteFloat2_0.Value).Value
   )
   n1 = InitVariable.Bool()
   TriggerRouteTriggerRouteBool2_0 = "TriggerRoute.Trigger Route (Bool, 2):Bool"(Set 0=^Fade, Set 1=^ReverseFade) {also {defaults {Value 1="<MetaSoundBuilderResult.SUCCEEDED: 0>"}}}
   TriggerCompareClamp_0 = TriggerCompare.Clamp:Bool(
    Compare=@ADSREnvelopeADSREnvelope_0.On Decay Triggered
    A=VariableAccessor.Bool(Variable=VariableMutator.Bool(Variable=@n1.Variable, Value=@TriggerRouteTriggerRouteBool2_0.Value).Variable).Value
   )
  ]
  links [
   {dst="@On Trigger", src=ADSREnvelopeADSREnvelope_0.On Attack Triggered}
   {dst="@Out Value", src=MapRangeMapRange_0.Out Value}
   {dst="@OnDoneReverseFade", src=TriggerCompareClamp_0.False}
   {dst="@OnDoneFade", src=TriggerCompareClamp_0.True}
  ]
 }]
