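# MetaSound patch that derives a no-repeat count from a wave array's size, likely for random selection that avoids recent repeats.
# On InTrigger, it subtracts two from the array length and compares the result with zero.
# A trigger router sets NoRepeatNumber to that difference, or to a fallback of one for small arrays, then fires OutTrigger.
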MetaSoundPatch /FortMetasound/Patches/MSP_GetNoRepeatNumber
inputs [
 InTrigger {type=Trigger}
 WaveArray {type=WaveAsset:Array}
]
outputs [
 NoRepeatNumber {type=Int32}
 OutTrigger {type=Trigger}
]
pages [Default {
  input_defaults {InTrigger=null, WaveArray []}
  nodes [
   n1 = UE.Subtract:Int32(PrimaryOperand=Array.Num:WaveAsset:Array(Array=^WaveArray).Num) {also {defaults {AdditionalOperands=2}}}
   n2 = TriggerCompare.Clamp:Int32(Compare=^InTrigger, A=@n1.out) {also {defaults {B=0, Type=4}}}
   TriggerRouteTriggerRouteInt322_0 = "TriggerRoute.Trigger Route (Int32, 2):Int32"(Value 1=UE.Reroute(Value=UE.Reroute(Value=@n1.out).Value).Value, Set 1=@n2.FALSE, Set 0=@n2.TRUE) {also {defaults {Value 0=1}}}
  ]
  links [{dst="@OutTrigger", src=TriggerRouteTriggerRouteInt322_0.On Set}, {dst="@NoRepeatNumber", src=TriggerRouteTriggerRouteInt322_0.Value}]
 }]
