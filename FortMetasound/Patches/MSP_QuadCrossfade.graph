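# MetaSound patch that crossfades between two four-channel audio sources (front left, front right, side left, side right).
# Each channel has its own two-input audio crossfade node, pairing the matching channel of Input 1 and Input 2.
# A single FloatIn value, defaulting to fully Input 1, drives all four crossfades together, feeding the matching front and side outputs.
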MetaSoundPatch /FortMetasound/Patches/MSP_QuadCrossfade
inputs [
 FloatIn {type=Float}
 Input 1 FL {type=Audio}
 Input 1 FR {type=Audio}
 Input 1 SL {type=Audio}
 Input 1 SR {type=Audio}
 Input 2 FL {type=Audio}
 Input 2 FR {type=Audio}
 Input 2 SL {type=Audio}
 Input 2 SR {type=Audio}
]
interfaces [MetaSound:1.0]
outputs [
 FrontLeft {type=Audio}
 FrontRight {type=Audio}
 OutSideLeft {type=Audio}
 OutSideRight {type=Audio}
]
pages [Default {
  input_defaults {
   FloatIn=0
   Input 1 FL=null
   Input 1 FR=null
   Input 1 SL=null
   Input 1 SR=null
   Input 2 FL=null
   Input 2 FR=null
   Input 2 SL=null
   Input 2 SR=null
  }
  nodes [
   CrossfadeTriggerRouteAudio2_0 = "Crossfade.Trigger Route (Audio, 2):Audio"(In 0=^Input 1 SR, In 1=^Input 2 SR, Crossfade Value=UE.Reroute(Value=^FloatIn).Value)
   CrossfadeTriggerRouteAudio2_1 = "Crossfade.Trigger Route (Audio, 2):Audio"(Crossfade Value=^FloatIn, In 0=^Input 1 SL, In 1=^Input 2 SL)
   CrossfadeTriggerRouteAudio2_2 = "Crossfade.Trigger Route (Audio, 2):Audio"(In 0=^Input 1 FR, In 1=^Input 2 FR, Crossfade Value=UE.Reroute(Value=^FloatIn).Value)
   CrossfadeTriggerRouteAudio2_3 = "Crossfade.Trigger Route (Audio, 2):Audio"(In 0=^Input 1 FL, In 1=^Input 2 FL, Crossfade Value=UE.Reroute(Value=^FloatIn).Value)
  ]
  links [
   {dst="@OutSideRight", src=CrossfadeTriggerRouteAudio2_0.Out}
   {dst="@OutSideLeft", src=CrossfadeTriggerRouteAudio2_1.Out}
   {dst="@FrontRight", src=CrossfadeTriggerRouteAudio2_2.Out}
   {dst="@FrontLeft", src=CrossfadeTriggerRouteAudio2_3.Out}
  ]
 }]
