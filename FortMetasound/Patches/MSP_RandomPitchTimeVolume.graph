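# MetaSound patch that, on a trigger, outputs random pitch in semitones, random time and random linear volume gain for varying sounds.
# The trigger chains three random nodes in sequence: pitch, then delay time, then volume, each bounded by min and max inputs.
# Decibel limits convert to linear gain before randomizing, and are wired crosswise, with the minimum feeding Max and the maximum feeding Min.
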MetaSoundPatch /FortMetasound/Patches/MSP_RandomPitchTimeVolume
inputs [
 "Delay Max (Seconds)" {type=Time}
 "Delay Min (Seconds)" {type=Time}
 "Random Pitch Max (Semitones)" {type=Float}
 "Random Pitch Min (Semitones)" {type=Float}
 "Random Volume Change Max (Decibels)" {type=Float}
 "Random Volume Change Min (Decibels)" {type=Float}
 Trigger {type=Trigger}
]
outputs [
 Random Pitch {type=Float}
 Random Time {type=Time}
 Random Volume {type=Float}
]
pages [Default {
  input_defaults {
   "Delay Max (Seconds)"=0
   "Delay Min (Seconds)"=0
   "Random Pitch Max (Semitones)"=1
   "Random Pitch Min (Semitones)"=-1
   "Random Volume Change Max (Decibels)"=1
   "Random Volume Change Min (Decibels)"=-1
   Trigger=null
  }
  nodes [
   DecibelstoLinearGain_0 = UE.Decibels to Linear Gain:Float(Decibels=^"Random Volume Change Min (Decibels)")
   DecibelstoLinearGain_1 = UE.Decibels to Linear Gain:Float(Decibels=^"Random Volume Change Max (Decibels)")
   RandomFloat_1 = UE.RandomFloat(Max=^"Random Pitch Max (Semitones)", Min=^"Random Pitch Min (Semitones)", Next=^Trigger)
   RandomTime_0 = UE.RandomTime(Max=^"Delay Max (Seconds)", Min=^"Delay Min (Seconds)", Next=@RandomFloat_1.On Next)
   RandomFloat_0 = UE.RandomFloat(Max=@DecibelstoLinearGain_0.Linear Gain, Min=@DecibelstoLinearGain_1.Linear Gain, Next=@RandomTime_0.On Next)
  ]
  links [
   {dst="@Random Volume", src=RandomFloat_0.Value}
   {dst="@Random Pitch", src=RandomFloat_1.Value}
   {dst="@Random Time", src=RandomTime_0.Value}
  ]
 }]
