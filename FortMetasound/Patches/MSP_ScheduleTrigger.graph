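# MetaSound patch that schedules a single output trigger to fire a chosen number of incoming triggers after Schedule fires.
# Schedule stores the current In Trigger count plus Triggers From Now (minimum one) as the target; Cancel resets it to -1.
# Each In Trigger compares count against target, opening a trigger gate on a match.
# The gate's output feeds back through a deferred variable to close it.
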MetaSoundPatch /FortMetasound/Patches/MSP_ScheduleTrigger
inputs [
 Cancel {type=Trigger}
 In Trigger {type=Trigger}
 Schedule {type=Trigger}
 Triggers From Now {type=Int32}
]
outputs [Scheduled Trigger {type=Trigger}]
pages [Default {
  input_defaults {Cancel=null, In Trigger=null, Schedule=null, Triggers From Now=1}
  nodes [
   TriggerCounter_0 = UE.Trigger Counter(In=^In Trigger)
   n1 = VariableAccessor.Int32(Variable=VariableMutator.Int32(Variable=InitVariable.Int32().Variable, Value=@TriggerCounter_0.Count).Variable)
   n2 = VariableAccessor.Int32(Variable=@n1.Variable)
   n3 = UE.Add:Int32(AdditionalOperands=Max.Max:Int32(A=^Triggers From Now) {also {defaults {B=1}}}.Value, PrimaryOperand=@n1.Value)
   n4 = VariableAccessor.Int32(Variable=VariableMutator.Int32(
     Variable=InitVariable.Int32().Variable
     Value=Value.Value:Int32(Reset=^Cancel, Set=^Schedule, Target Value=@n3.Out) {also {defaults {Init Value=-1}}}.Output Value
    ).Variable)
   n5 = VariableDeferredAccessor.Trigger(Variable=InitVariable.Trigger().Variable)
   TriggerControl_0 = UE.Trigger Control(
    Trigger In=^In Trigger
    Open=UE.Reroute(Value=TriggerCompare.Clamp:Int32(Compare=^In Trigger, B=@n2.Value, A=@n4.Value).True).Value
    Close=@n5.Value
   )
   VariableMutator.Trigger(Value=@TriggerControl_0.Trigger Out, Variable=@n5.Variable)
  ]
  links [{dst="@Scheduled Trigger", src=TriggerControl_0.Trigger Out}]
 }]
