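# MSP_StereoGain is a MetaSound patch that applies one linear gain to a stereo audio pair.
# Left In and Right In each pass through their own audio-by-float multiply driven by the shared Gain (Linear) input, feeding Left Out and Right Out.
# The gain defaults to unity, so the signal passes unchanged unless that input is set.
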MetaSoundPatch /FortMetasound/Patches/MSP_StereoGain
inputs [
 "Gain (Linear)" {type=Float}
 Left In {type=Audio}
 Right In {type=Audio}
]
outputs [
 Left Out {type=Audio}
 Right Out {type=Audio}
]
pages [Default {
  input_defaults {"Gain (Linear)"=1, Left In=null, Right In=null}
  nodes [
   Multiply_0 = UE.Multiply:Audio by Float(AdditionalOperands=^"Gain (Linear)", PrimaryOperand=^Right In)
   Multiply_1 = UE.Multiply:Audio by Float(AdditionalOperands=^"Gain (Linear)", PrimaryOperand=^Left In)
  ]
  links [{dst="@Right Out", src=Multiply_0.Out}, {dst="@Left Out", src=Multiply_1.Out}]
 }]
