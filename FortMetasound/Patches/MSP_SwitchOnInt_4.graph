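# MetaSound patch that acts as a four-way switch, routing an incoming Compare trigger to Out 0 through Out 3 according to an integer input.
# Chained Trigger Compare nodes test Int against 0, 1, 2 and 3 in order, each firing its matching output on true and passing the trigger onward on false.
# The last false branch is unconnected, so a trigger that fails every test produces no output.
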MetaSoundPatch /FortMetasound/Patches/MSP_SwitchOnInt_4
inputs [
 Compare {type=Trigger}
 Int {type=Int32}
]
outputs [
 Out 0 {type=Trigger}
 Out 1 {type=Trigger}
 Out 2 {type=Trigger}
 Out 3 {type=Trigger}
]
pages [Default {
  input_defaults {Compare=null, Int=0}
  nodes [
   TriggerCompareClamp_2 = TriggerCompare.Clamp:Int32(Compare=^Compare, A=^Int) {also {defaults {B=0}}}
   TriggerCompareClamp_0 = TriggerCompare.Clamp:Int32(A=^Int, Compare=@TriggerCompareClamp_2.FALSE) {also {defaults {B=1}}}
   TriggerCompareClamp_3 = TriggerCompare.Clamp:Int32(A=^Int, Compare=@TriggerCompareClamp_0.FALSE) {also {defaults {B=2}}}
   TriggerCompareClamp_1 = TriggerCompare.Clamp:Int32(A=^Int, Compare=@TriggerCompareClamp_3.FALSE) {also {defaults {B=3}}}
  ]
  links [
   {dst="@Out 1", src=TriggerCompareClamp_0.TRUE}
   {dst="@Out 3", src=TriggerCompareClamp_1.TRUE}
   {dst="@Out 0", src=TriggerCompareClamp_2.TRUE}
   {dst="@Out 2", src=TriggerCompareClamp_3.TRUE}
  ]
 }]
