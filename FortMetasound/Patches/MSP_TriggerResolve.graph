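# MetaSound patch that arbitrates between triggers A and B when they fire together, passing lone triggers through to their matching outputs.
# Each input briefly sets a Bool that a deferred trigger variable resets, letting chained Trigger Compare nodes detect coincidence and fire OnResolved.
# On a collision, Execute A First selects the winning output; Execute Both also fires the other, delayed through a deferred variable accessor.
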MetaSoundPatch /FortMetasound/Patches/MSP_TriggerResolve
inputs [
 A {type=Trigger}
 B {type=Trigger}
 Execute A First {type=Bool}
 Execute Both {type=Bool}
]
outputs [
 A {type=Trigger}
 B {type=Trigger}
 OnResolved {type=Trigger}
]
pages [Default {
  input_defaults {A=null, B=null, Execute A First="<MetaSoundBuilderResult.SUCCEEDED: 0>", Execute Both="<MetaSoundBuilderResult.SUCCEEDED: 0>"}
  nodes [
   n1 = InitVariable.Trigger()
   TriggerAnyTriggerAny2_1 = "TriggerAny.Trigger Any (2)"(In 0=^A, In 1=^B)
   n2 = InitVariable.Bool()
   n3 = VariableDeferredAccessor.Trigger(Variable=InitVariable.Trigger().Variable)
   n4 = Value.Value:Bool(Set=^A, Reset=@n3.Value)
   n5 = TriggerCompare.Clamp:Bool(
    Compare=@TriggerAnyTriggerAny2_1.Out
    A=VariableAccessor.Bool(Variable=VariableMutator.Bool(Variable=@n2.Variable, Value=@n4.Output Value).Variable).Value
   ) {also {defaults {B="<MetaSoundBuilderResult.SUCCEEDED: 0>"}}}
   n6 = VariableDeferredAccessor.Trigger(Variable=InitVariable.Trigger().Variable)
   n7 = Value.Value:Bool(Set=^B, Reset=@n6.Value)
   n8 = VariableAccessor.Bool(Variable=VariableMutator.Bool(Variable=InitVariable.Bool().Variable, Value=@n7.Output Value).Variable)
   TriggerCompareClamp_2 = TriggerCompare.Clamp:Bool(Compare=@n5.True, A=@n8.Value) {also {defaults {B="<MetaSoundBuilderResult.SUCCEEDED: 0>"}}}
   n9 = InitVariable.Trigger()
   n10 = InitVariable.Trigger()
   n11 = TriggerCompare.Clamp:Bool(
    A=^Execute Both
    Compare=VariableAccessor.Trigger(Variable=VariableMutator.Trigger(Variable=@n10.Variable, Value=@TriggerCompareClamp_2.True).Variable).Value
   ) {also {defaults {B="<MetaSoundBuilderResult.SUCCEEDED: 0>"}}}
   n12 = TriggerCompare.Clamp:Bool(A=^Execute A First, Compare=@n11.True) {also {defaults {B="<MetaSoundBuilderResult.SUCCEEDED: 0>"}}}
   n13 = TriggerCompare.Clamp:Bool(A=^Execute A First, Compare=@n11.False) {also {defaults {B="<MetaSoundBuilderResult.SUCCEEDED: 0>"}}}
   TriggerAnyTriggerAny2_2 = "TriggerAny.Trigger Any (2)"(In 0=@n12.True, In 1=@n13.True)
   n14 = VariableDeferredAccessor.Trigger(Variable=InitVariable.Trigger().Variable)
   TriggerAnyTriggerAny3_0 = "TriggerAny.Trigger Any (3)"(
    In 1=VariableAccessor.Trigger(Variable=VariableMutator.Trigger(Variable=@n1.Variable, Value=@TriggerCompareClamp_2.False).Variable).Value
    In 0=VariableAccessor.Trigger(Variable=VariableMutator.Trigger(Variable=@n9.Variable, Value=@TriggerAnyTriggerAny2_2.Out).Variable).Value
    In 2=@n14.Value
   )
   n15 = InitVariable.Trigger()
   TriggerAnyTriggerAny2_0 = "TriggerAny.Trigger Any (2)"(In 0=@n12.False, In 1=@n13.False)
   n16 = InitVariable.Trigger()
   n17 = TriggerCompare.Clamp:Bool(Compare=@n5.False, A=VariableAccessor.Bool(Variable=@n8.Variable).Value) {also {defaults {B="<MetaSoundBuilderResult.SUCCEEDED: 0>"}}}
   n18 = VariableDeferredAccessor.Trigger(Variable=InitVariable.Trigger().Variable)
   TriggerAnyTriggerAny3_1 = "TriggerAny.Trigger Any (3)"(
    In 0=VariableAccessor.Trigger(Variable=VariableMutator.Trigger(Variable=@n15.Variable, Value=@TriggerAnyTriggerAny2_0.Out).Variable).Value
    In 1=VariableAccessor.Trigger(Variable=VariableMutator.Trigger(Variable=@n16.Variable, Value=@n17.True).Variable).Value
    In 2=@n18.Value
   )
   VariableMutator.Trigger(Value=@n7.On Set, Variable=@n6.Variable)
   VariableMutator.Trigger(Value=@n12.False, Variable=@n14.Variable)
   VariableMutator.Trigger(Value=@n4.On Set, Variable=@n3.Variable)
   VariableMutator.Trigger(Value=@n12.True, Variable=@n18.Variable)
  ]
  links [
   {dst="@A", src=TriggerAnyTriggerAny3_0.Out}
   {dst="@B", src=TriggerAnyTriggerAny3_1.Out}
   {dst="@OnResolved", src=TriggerCompareClamp_2.True}
  ]
 }]
