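# MetaSound patch that takes a wave asset and outputs a random time within its duration, named for use as a random playback start time.
# It reads the wave's duration and feeds it to a RandomTime node as the upper bound.
# The duration is also converted to float for a value-change trigger, which draws a new random time whenever the duration changes.
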MetaSoundPatch /FortMetasound/Patches/MSP_WaveGetRandomStartTime
inputs [Sound {type=WaveAsset}]
interfaces [MetaSound:1.0]
outputs [Value {type=Time}]
pages [Default {
  input_defaults {Sound=null}
  nodes [
   GetWaveDuration_0 = "UE.Get Wave Duration: "(Wave=^Sound)
   TriggerOnValueChange_0 = UE.Trigger On Value Change:Float(Value=UE.ConversionTimeToFloat(In=@GetWaveDuration_0.Duration).Out)
   RandomTime_0 = UE.RandomTime(Max=UE.Reroute(Value=UE.Reroute(Value=@GetWaveDuration_0.Duration).Value).Value, Next=@TriggerOnValueChange_0.Trigger)
  ]
  links [{dst="@Value", src=RandomTime_0.Value}]
 }]
