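# MetaSound patch that maps a float input range to a linear gain output, with the output range specified in decibels.
# Out Range A and B each pass through a Decibels to Linear Gain node, then feed a MapRange node along with the input, input range and Clamped flag.
# Because conversion happens before mapping, the interpolation runs in linear gain rather than in decibels.
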MetaSoundPatch /FortMetasound/Patches/MapRange_DecibelOutput
inputs [
 Clamped {type=Bool}
 In {type=Float}
 In Range A {type=Float}
 In Range B {type=Float}
 Out Range A {type=Float}
 Out Range B {type=Float}
]
interfaces [MetaSound:1.0]
outputs [Out Value {type=Float}]
pages [Default {
  input_defaults {Clamped="<MetaSoundBuilderResult.SUCCEEDED: 0>", In=0, In Range A=0, In Range B=1, Out Range A=0, Out Range B=0}
  nodes [
   DecibelstoLinearGain_0 = UE.Decibels to Linear Gain:Float(Decibels=^Out Range A)
   DecibelstoLinearGain_1 = UE.Decibels to Linear Gain:Float(Decibels=^Out Range B)
   MapRangeMapRange_0 = MapRange.MapRange:Float(
    Clamped=^Clamped
    In=^In
    In Range A=^In Range A
    In Range B=^In Range B
    Out Range A=@DecibelstoLinearGain_0.Linear Gain
    Out Range B=@DecibelstoLinearGain_1.Linear Gain
   )
  ]
  links [{dst="@Out Value", src=MapRangeMapRange_0.Out Value}]
 }]
