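# MetaSound patch that mixes two 5.1 audio sources into a single 5.1 output, with a linear volume per source.
# Three two-input stereo mixers handle the channels in pairs: left/right, center/LFE and surround left/right.
# Each source's volume drives the matching gain on all three mixers, scaling its six channels together, and defaults to unity.
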MetaSoundPatch /FortMetasound/Patches/Mixer_51_Two
inputs [
 1.InC {type=Audio}
 1.InL {type=Audio}
 1.InLF {type=Audio}
 1.InR {type=Audio}
 1.InSL {type=Audio}
 1.InSR {type=Audio}
 "1.Volume(Lin)" {type=Float}
 2.InC {type=Audio}
 2.InL {type=Audio}
 2.InLF {type=Audio}
 2.InR {type=Audio}
 2.InSL {type=Audio}
 2.InSR {type=Audio}
 "2.Volume(Lin)" {type=Float}
]
interfaces [MetaSound:1.0]
outputs [
 OutC {type=Audio}
 OutL {type=Audio}
 OutLF {type=Audio}
 OutR {type=Audio}
 OutSL {type=Audio}
 OutSR {type=Audio}
]
pages [Default {
  input_defaults {
   1.InC=null
   1.InL=null
   1.InLF=null
   1.InR=null
   1.InSL=null
   1.InSR=null
   "1.Volume(Lin)"=1
   2.InC=null
   2.InL=null
   2.InLF=null
   2.InR=null
   2.InSL=null
   2.InSR=null
   "2.Volume(Lin)"=1
  }
  nodes [
   AudioMixerAudioMixerStereo2_0 = "AudioMixer.Audio Mixer (Stereo, 2)"(In 0 L=^1.InL, In 0 R=^1.InR, Gain 0=^"1.Volume(Lin)", In 1 L=^2.InL, In 1 R=^2.InR, Gain 1=^"2.Volume(Lin)")
   AudioMixerAudioMixerStereo2_1 = "AudioMixer.Audio Mixer (Stereo, 2)"(In 0 L=^1.InC, In 0 R=^1.InLF, Gain 0=^"1.Volume(Lin)", In 1 L=^2.InC, In 1 R=^2.InLF, Gain 1=^"2.Volume(Lin)")
   AudioMixerAudioMixerStereo2_2 = "AudioMixer.Audio Mixer (Stereo, 2)"(In 0 L=^1.InSL, In 0 R=^1.InSR, Gain 0=^"1.Volume(Lin)", In 1 L=^2.InSL, In 1 R=^2.InSR, Gain 1=^"2.Volume(Lin)")
  ]
  links [
   {dst="@OutL", src=AudioMixerAudioMixerStereo2_0.Out L}
   {dst="@OutR", src=AudioMixerAudioMixerStereo2_0.Out R}
   {dst="@OutC", src=AudioMixerAudioMixerStereo2_1.Out L}
   {dst="@OutLF", src=AudioMixerAudioMixerStereo2_1.Out R}
   {dst="@OutSL", src=AudioMixerAudioMixerStereo2_2.Out L}
   {dst="@OutSR", src=AudioMixerAudioMixerStereo2_2.Out R}
  ]
 }]
