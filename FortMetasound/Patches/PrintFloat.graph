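# MetaSound patch that prints a labeled float to the log as a debugging aid, with no outputs.
# A print fires from the manual Trigger input or, when bTriggerOnValueChange allows it, from a change in the Input value.
# Both sources merge through a Trigger Any, then pass a bBypass compare gate before reaching the Print Log node.
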MetaSoundPatch /FortMetasound/Patches/PrintFloat
inputs [
 bBypass {type=Bool}
 bTriggerOnValueChange {type=Bool}
 Input {type=Float}
 Label {type=String}
 Trigger {type=Trigger}
]
interfaces [MetaSound:1.0]
outputs []
pages [Default {
  input_defaults {Input=0, Label=Print Float:, Trigger=null, bBypass=null, bTriggerOnValueChange="<MetaSoundBuilderResult.SUCCEEDED: 0>"}
  nodes [
   TriggerOnValueChange_0 = UE.Trigger On Value Change:Float(Value=^Input)
   TriggerAnyTriggerAny2_0 = "TriggerAny.Trigger Any (2)"(
    In 0=^Trigger
    In 1=TriggerCompare.Clamp:Bool(A=^bTriggerOnValueChange, Compare=@TriggerOnValueChange_0.Trigger) {also {defaults {B="<MetaSoundBuilderResult.SUCCEEDED: 0>"}}}.True
   )
   PrintLogPrintLog_0 = Print Log.Print Log:Float(Value To Log=^Input, Label=^Label, Trigger=TriggerCompare.Clamp:Bool(A=^bBypass, Compare=@TriggerAnyTriggerAny2_0.Out).True)
  ]
 }]
