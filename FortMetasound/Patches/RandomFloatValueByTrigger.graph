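# MetaSound patch that outputs a random float between FloatMin and FloatMax, re-rolled on each TriggerIn.
# A Random Float node feeds Output directly; its default bounds are inverted, with the minimum above the maximum.
# For debugging, a value-change trigger passes through a Trigger Compare against DebugPrint to log the new value with the label "Output".
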MetaSoundPatch /FortMetasound/Patches/RandomFloatValueByTrigger
inputs [
 DebugPrint {type=Bool}
 FloatMax {type=Float}
 FloatMin {type=Float}
 TriggerIn {type=Trigger}
]
interfaces [MetaSound:1.0]
outputs [Output {type=Float}]
pages [Default {
  input_defaults {DebugPrint=null, FloatMax=0.2, FloatMin=1, TriggerIn=null}
  nodes [
   RandomFloat_0 = UE.RandomFloat(Max=^FloatMax, Min=^FloatMin, Next=^TriggerIn)
   TriggerOnValueChange_0 = UE.Trigger On Value Change:Float(Value=@RandomFloat_0.Value)
   PrintLogPrintLog_0 = Print Log.Print Log:Float(
    Value To Log=@RandomFloat_0.Value
    Trigger=TriggerCompare.Clamp:Bool(B=^DebugPrint, Compare=@TriggerOnValueChange_0.Trigger) {also {defaults {A="<MetaSoundBuilderResult.SUCCEEDED: 0>"}}}.TRUE
   ) {also {defaults {Label=Output}}}
  ]
  links [{dst="@Output", src=RandomFloat_0.Value}]
 }]
