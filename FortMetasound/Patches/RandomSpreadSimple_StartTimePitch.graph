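# MetaSound patch that generates random pitch shift and start time values for varying sound playback.
# The absolute value of each spread input feeds an Add and a Subtract, forming a symmetric range around zero, so the sign of the spread does not matter.
# The In trigger advances a random float and a random time generator, which drive the pitch shift and start time outputs.
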MetaSoundPatch /FortMetasound/Patches/RandomSpreadSimple_StartTimePitch
inputs [
 In {type=Trigger}
 PitchShiftSpread {type=Float}
 StartTimeSpread {type=Time}
]
outputs [
 RandomPitchShiftOut {type=Float}
 StartTimeOut {type=Time}
]
pages [Default {
  input_defaults {In=null, PitchShiftSpread=0.05, StartTimeSpread=0}
  nodes [
   n1 = AbsoluteValue.Abs:Float(Input=^PitchShiftSpread)
   RandomFloat_0 = UE.RandomFloat(Next=^In, Max=UE.Add:Float(AdditionalOperands=@n1.Value).Out, Min=UE.Subtract:Float(AdditionalOperands=@n1.Value).Out)
   n2 = AbsoluteValue.Abs:Time(Input=^StartTimeSpread)
   RandomTime_0 = UE.RandomTime(Next=^In, Max=UE.Add:Time(AdditionalOperands=@n2.Value).Out, Min=UE.Subtract:Time(AdditionalOperands=@n2.Value) {also {version [1,1]}}.Out)
  ]
  links [{dst="@RandomPitchShiftOut", src=RandomFloat_0.Value}, {dst="@StartTimeOut", src=RandomTime_0.Value}]
 }]
