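# StereoMixer_One is a MetaSound patch that applies a single linear volume to one stereo audio input.
# The left and right inputs are each multiplied by the same Volume(Lin) float, which defaults to unity gain, and sent to the matching left and right outputs.
# It has no summing, panning or smoothing, so it works as a one-channel stereo gain stage.
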MetaSoundPatch /FortMetasound/Patches/StereoMixer_One
inputs [
 InL {type=Audio}
 InR {type=Audio}
 "Volume(Lin)" {type=Float}
]
interfaces [MetaSound:1.0]
outputs [
 OutL {type=Audio}
 OutR {type=Audio}
]
pages [Default {
  input_defaults {InL=null, InR=null, "Volume(Lin)"=1}
  nodes [
   Multiply_0 = UE.Multiply:Audio by Float(PrimaryOperand=^InL, AdditionalOperands=^"Volume(Lin)")
   Multiply_1 = UE.Multiply:Audio by Float(PrimaryOperand=^InR, AdditionalOperands=^"Volume(Lin)")
  ]
  links [{dst="@OutL", src=Multiply_0.Out}, {dst="@OutR", src=Multiply_1.Out}]
 }]
