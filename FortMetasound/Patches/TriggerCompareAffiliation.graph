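# MetaSound patch that compares two Fortnite audio affiliation values when triggered.
# It converts enum inputs A and B to Int32 and feeds them, with the Compare trigger and selectable comparison Type, into a single integer Trigger Compare node.
# That node's result fires either the True or False output trigger.
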MetaSoundPatch /FortMetasound/Patches/TriggerCompareAffiliation
inputs [
 A {type=Enum:FortAudioAffiliation}
 B {type=Enum:FortAudioAffiliation}
 Compare {type=Trigger}
 Type {type=Enum:TriggerComparisonType}
]
outputs [
 False {type=Trigger}
 True {type=Trigger}
]
pages [Default {
  input_defaults {A=0, B=0, Compare=null, Type=0}
  nodes [TriggerCompareClamp_0 = TriggerCompare.Clamp:Int32(
    Compare=^Compare
    Type=^Type
    B=Convert.Int32:Enum:FortAudioAffiliation("Enum:FortAudioAffiliation"=^B).Int32
    A=Convert.Int32:Enum:FortAudioAffiliation("Enum:FortAudioAffiliation"=^A).Int32
   )]
  links [{dst="@False", src=TriggerCompareClamp_0.FALSE}, {dst="@True", src=TriggerCompareClamp_0.TRUE}]
 }]
